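# Post-process style Control Rig for the M_MED_YokeQuick body that runs the shared DeformRig function, then corrects suit jacket and tie bones.
# Jacket offsets are weighted by thigh-up anim attributes, thigh rotations, spine bend, pelvis height and interpolated crouch, slide and DBNO flags.
# The tie is parent-constrained across upper-body bones.
# Unless isFE_CR is set, spherical pose readers on spine and head, plus DBNO state, adjust it further.
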ControlRigBlueprint /BRCosmetics/Characters/Player/Male/Medium/Bodies/M_MED_YokeQuick/Meshes/M_MED_YokeQuick_ControlRig
graphs [
 New Function_ContainedGraph {nodes [
   RigVMFunctionEntryNode()
   > RigVMFunction_Sequence = RigVMAggregateNode()
   &RigUnit_ModifyTransforms(
    ItemToModify="((Item=(Type=Bone,Name=\"base_suitSideFront_l\"),Transform=(Rotation=(X=0.1257,Y=0.0926,Z=-0.3485,W=0.9242),Translation=(X=0,Y=0,Z=0),Scale3D=(X=1,Y=1,Z=1))),(Item=(Type=Bone,Name=\"base_suitSide_l\"),Transform=(Rotation=(X=-0.2616,Y=0.0715,Z=0.0326,W=-0.962), ...(truncated)"
    Weight=RigVMDispatchNode(BoneName=root, Default=1, Name=deformRig_l_thigh_up_alpha) {template="DISPATCH_RigDispatch_GetAnimAttribute(in Name,in BoneName,in Default,out Value,out Found)"}.Value
    ExecutePin=@RigVMFunction_Sequence.A
   )
   > &RigUnit_ModifyTransforms(
    ItemToModify="((Item=(Type=Bone,Name=\"base_suitSideFront_r\"),Transform=(Rotation=(X=0.1391,Y=-0.0675,Z=0.4296,W=-0.8897),Translation=(X=0,Y=1.9999,Z=0),Scale3D=(X=1,Y=1,Z=1))),(Item=(Type=Bone,Name=\"base_suitSide_r\"),Transform=(Rotation=(X=0.0674,Y=-0.2474,Z=0.0295,W=-0 ...(truncated)"
    Weight=RigVMDispatchNode(BoneName=root, Default=1, Name=deformRig_r_thigh_up_alpha) {template="DISPATCH_RigDispatch_GetAnimAttribute(in Name,in BoneName,in Default,out Value,out Found)"}.Value
   )
   Delta Rotation Y = RigVMFunctionReferenceNode(Child=thigh_l, Parent=thigh_l_null, Source Maximum=-90, Source Minimum=-10, ExecuteContext=@RigVMFunction_Sequence.B)
   > RigVMVariableNode(Value=@"Delta Rotation Y".Delta Rotation) {variable=Thigh_L_SideRotation}
   > Delta Rotation Y_1 = RigVMFunctionReferenceNode(Child=thigh_r, Parent=thigh_r_null, Source Maximum=-80, Source Minimum=-10)
   > RigVMVariableNode(Value=@"Delta Rotation Y_1".Delta Rotation) {variable=Thigh_R_SideRotation}
   n1 = &RigVMFunction_AlphaInterp(
    InterpSpeedDecreasing=10
    InterpSpeedIncreasing=10
    Scale=1
    bClampResult=true
    Value=&RigVMFunction_MathBoolToFloat(Value=RigVMVariableNode() {variable=isCrouching_CR}.Value).Result
   )
   Remap = &RigVMFunction_MathDoubleRemap(SourceMaximum=1, SourceMinimum=0, TargetMaximum=0, TargetMinimum=1, Value=@n1.Result)
   Multiply = &RigVMFunction_MathDoubleMul(B=@Remap.Result, A=RigVMVariableNode() {variable=Thigh_L_SideRotation}.Value)
   > &RigUnit_ModifyTransforms(
    ItemToModify="((Item=(Type=Bone,Name=\"base_suitSide_l\"),Transform=(Rotation=(X=-0.0652,Y=0.126,Z=-0.3079,W=0.9408),Translation=(X=0,Y=0,Z=0),Scale3D=(X=1,Y=1,Z=1))),(Item=(Type=Bone,Name=\"base_suitSideFront_l\"),Transform=(Rotation=(X=0.0529,Y=0.2994,Z=-0.0636,W=0.9505), ...(truncated)"
    Weight=@Multiply.Result
   )
   > &RigUnit_ModifyTransforms(
    ItemToModify="((Item=(Type=Bone,Name=\"base_suitSide_r\"),Transform=(Rotation=(X=0.027,Y=-0.1384,Z=-0.6043,W=0.7842),Translation=(X=0,Y=0,Z=0),Scale3D=(X=1,Y=1,Z=1))),(Item=(Type=Bone,Name=\"base_suitSideFront_r\"),Transform=(Rotation=(X=-0.0698,Y=-0.3101,Z=-0.3795,W=0.8689 ...(truncated)"
    Weight=&RigVMFunction_MathDoubleMul(
     B=&RigVMFunction_MathDoubleRemap(SourceMaximum=1, SourceMinimum=0, TargetMaximum=0, TargetMinimum=1, Value=@n1.Result).Result
     A=RigVMVariableNode() {variable=Thigh_R_SideRotation}.Value
    ).Result
   )
   &RigUnit_ModifyTransforms(
    ItemToModify="((Item=(Type=Bone,Name=\"base_suitBack\"),Transform=(Rotation=(X=0,Y=0,Z=0.0423,W=-0.9991),Translation=(X=0,Y=0,Z=0),Scale3D=(X=1,Y=1,Z=1))),(Item=(Type=Bone,Name=\"base_suitSideBack_r\"),Transform=(Rotation=(X=0,Y=0,Z=-0.0285,W=0.9996),Translation=(X=0,Y=-0.4 ...(truncated)"
    Weight=@n1.Result
    ExecutePin=@RigVMFunction_Sequence.C
   )
   > &RigUnit_ModifyTransforms(
    ItemToModify="((Item=(Type=Bone,Name=\"dyn_suitSideFront_02_l\"),Transform=(Rotation=(X=0.0861,Y=-0.0936,Z=0.2635,W=0.9562),Translation=(X=0,Y=0,Z=0),Scale3D=(X=1,Y=1,Z=1))),(Item=(Type=Bone,Name=\"dyn_suitSideFront_02_r\"),Transform=(Rotation=(X=0,Y=0,Z=0.1615,W=0.9869),Tr ...(truncated)"
    Weight=@n1.Result
   )
   > Delta Rotation Z = RigVMFunctionReferenceNode(Child=thigh_l, Parent=thigh_l_null, Source Maximum=-145, Source Minimum=-120)
   > &RigUnit_ModifyTransforms(
    ItemToModify="((Item=(Type=Bone,Name=\"base_suitSideFront_l\"),Transform=(Rotation=(X=0,Y=0,Z=0,W=1),Translation=(X=0,Y=0,Z=0),Scale3D=(X=1,Y=1,Z=1))),(Item=(Type=Bone,Name=\"base_suitSide_l\"),Transform=(Rotation=(X=0,Y=0,Z=0,W=1),Translation=(X=0,Y=0,Z=0),Scale3D=(X=1,Y=1 ...(truncated)"
    Weight=&RigVMFunction_MathDoubleMul(B=@n1.Result, A=@"Delta Rotation Z".Delta Rotation).Result
   )
   n2 = &RigUnit_GetRelativeTransformForItem(Child=(Type=Bone,Name="spine_05"), Parent=(Type=Bone,Name="pelvis"), bChildInitial=false, bParentInitial=false)
   n3 = &RigVMFunction_MathDoubleRemap(
    SourceMaximum=50
    SourceMinimum=0
    TargetMaximum=1
    TargetMinimum=0
    Value=&RigVMFunction_MathQuaternionToEuler(
     RotationOrder=ZXY
     Value=&RigVMFunction_MathQuaternionSwingTwist(TwistAxis=(X=0,Y=0,Z=1), Input=@n2.RelativeTransform.Rotation).Swing
    ).Result.X
   )
   n4 = &RigVMFunction_MathBoolToFloat(Value=RigVMVariableNode() {variable=isCrouching_CR}.Value)
   > &RigUnit_ModifyTransforms(
    ItemToModify="((Item=(Type=Bone,Name=\"base_suitSideFront_l\"),Transform=(Rotation=(X=0,Y=0,Z=-0.0343,W=0.9994),Translation=(X=0.0442,Y=-0.9552,Z=0),Scale3D=(X=1,Y=1,Z=1))),(Item=(Type=Bone,Name=\"dyn_suitSideFront_02_l\"),Transform=(Rotation=(X=0,Y=0,Z=-0.2228,W=0.9749),Tr ...(truncated)"
    WeightMinimum=0
    Weight=/Script/RigVM.RigVMFunction_AnimEasing(
     SourceMaximum=1
     SourceMinimum=0
     TargetMaximum=1
     TargetMinimum=0
     Type=CubicEaseOut
     Value=&RigVMFunction_MathDoubleMul(B=@n3.Result, A=@n4.Result).Result
    ) {is=struct RigVMUnitNode}.Result
   )
   n5 = &RigVMFunction_MathQuaternionToEuler(
    RotationOrder=ZXY
    Value=&RigVMFunction_MathQuaternionSwingTwist(
     TwistAxis=(X=0,Y=1,Z=0)
     Input=&RigUnit_GetRelativeTransformForItem(Child=(Type=Bone,Name="spine_05"), Parent=(Type=Bone,Name="pelvis"), bChildInitial=false, bParentInitial=false).RelativeTransform.Rotation
    ).Swing
   )
   > &RigUnit_ModifyTransforms(
    ItemToModify="((Item=(Type=Bone,Name=\"base_suitSideFront_l\"),Transform=(Rotation=(X=0,Y=0,Z=0,W=1),Translation=(X=0,Y=2,Z=0),Scale3D=(X=1,Y=1,Z=1))),(Item=(Type=Bone,Name=\"dyn_suitSideFront_02_l\"),Transform=(Rotation=(X=0,Y=0,Z=0,W=1),Translation=(X=0,Y=0,Z=0),Scale3D=( ...(truncated)"
    WeightMinimum=0
    Weight=&RigVMFunction_MathDoubleMul(
     B=&RigVMFunction_MathDoubleRemap(SourceMaximum=50, SourceMinimum=-1, TargetMaximum=1, TargetMinimum=0, Value=@n5.Result.Z).Result
     A=@n4.Result
    ).Result
   )
   n6 = &RigVMFunction_AlphaInterp(
    InterpSpeedDecreasing=10
    InterpSpeedIncreasing=10
    Scale=1
    bClampResult=true
    Value=&RigVMFunction_MathBoolToFloat(Value=RigVMVariableNode() {variable=isBaseballSliding_CR}.Value).Result
   )
   &RigUnit_ModifyTransforms(
    ItemToModify="((Item=(Type=Bone,Name=\"base_suitSideBack_r\"),Transform=(Rotation=(X=0,Y=0,Z=0,W=1),Translation=(X=0,Y=0,Z=0),Scale3D=(X=1,Y=1,Z=1))),(Item=(Type=Bone,Name=\"base_suitSide_r\"),Transform=(Rotation=(X=0,Y=0,Z=0,W=1),Translation=(X=0,Y=0,Z=0),Scale3D=(X=1,Y=1, ...(truncated)"
    Weight=@n6.Result
    ExecutePin=@RigVMFunction_Sequence.D
   )
   n7 = &RigVMFunction_AlphaInterp(
    InterpSpeedDecreasing=10
    InterpSpeedIncreasing=10
    Scale=1
    bClampResult=true
    Value=&RigVMFunction_MathBoolToFloat(Value=RigVMVariableNode() {variable=isDBNO_CR}.Value).Result
   )
   > &RigUnit_ModifyTransforms(
    ItemToModify="((Item=(Type=Bone,Name=\"base_suitSide_l\"),Transform=(Rotation=(X=0.0006,Y=0.1219,Z=-0.0047,W=-0.9925),Translation=(X=0,Y=0,Z=0),Scale3D=(X=1,Y=1,Z=1))),(Item=(Type=Bone,Name=\"base_suitSide_r\"),Transform=(Rotation=(X=0.0096,Y=-0.1074,Z=0.0887,W=-0.9902),Tra ...(truncated)"
    Weight=@n7.Result
   )
   n8 = &RigVMFunction_AlphaInterp(
    InterpSpeedDecreasing=10
    InterpSpeedIncreasing=10
    Scale=1
    bClampResult=true
    Value=&RigVMFunction_MathBoolToFloat(Value=RigVMVariableNode() {variable=isKneeSliding_CR}.Value).Result
   )
   n9 = &RigVMFunction_MathDoubleRemap(
    SourceMaximum=43.5
    SourceMinimum=89
    TargetMaximum=1
    TargetMinimum=0
    Value=&RigUnit_GetRelativeTransformForItem(Child=(Type=Bone,Name="pelvis"), Parent=(Type=Bone,Name="root"), bChildInitial=False, bParentInitial=False).RelativeTransform.Translation.Z
   )
   > &RigUnit_ModifyTransforms(
    ItemToModify="((Item=(Type=Bone,Name=\"base_suitBack\"),Transform=(Rotation=(X=0.0551,Y=0.1647,Z=-0.3125,W=0.9339),Translation=(X=-10,Y=0,Z=0),Scale3D=(X=1,Y=1,Z=1))),(Item=(Type=Bone,Name=\"base_suitSideBack_l\"),Transform=(Rotation=(X=0,Y=0,Z=-0.2164,W=0.9763),Translation ...(truncated)"
    WeightMinimum=0
    Weight=&RigVMFunction_AnimEasing(
     SourceMaximum=1
     SourceMinimum=0
     TargetMaximum=1
     TargetMinimum=0
     Type=CubicEaseOut
     Value=&RigVMFunction_MathDoubleMul(B=@n8.Result, A=@n9.Result).Result
    ).Result
   )
   RigVMFunctionReturnNode(ExecuteContext=@RigVMFunction_Sequence.E)
  ]}
 New Function_ContainedGraph {nodes [
   n1 = RigVMFunctionEntryNode()
   RigVMFunction_MathQuaternionToEuler = &RigVMFunction_MathQuaternionToEuler(
    RotationOrder=XYZ
    Value=&RigUnit_GetRelativeTransformForItem(
     Child=(Type=Bone,Name="hand_l")
     Parent=(Type=Null,Name="lowerarm_l")
     bChildInitial=False
     bParentInitial=False
     Child.Name=@n1.Child
     Parent.Name=@n1.Parent
    ).RelativeTransform.Rotation
   )
   Remap = &RigVMFunction_MathDoubleRemap(
    TargetMaximum=1
    TargetMinimum=0
    Value=@RigVMFunction_MathQuaternionToEuler.Result.Y
    SourceMaximum=@n1.Source Maximum
    SourceMinimum=@n1.Source Minimum
   )
   > RigVMFunctionReturnNode(Delta Rotation=@Remap.Result)
  ]}
 New Function_ContainedGraph {nodes [
   n1 = RigVMFunctionEntryNode()
   RigVMFunction_MathQuaternionToEuler = &RigVMFunction_MathQuaternionToEuler(
    RotationOrder=XYZ
    Value=&RigUnit_GetRelativeTransformForItem(
     Child=(Type=Bone,Name="hand_l")
     Parent=(Type=Null,Name="lowerarm_l")
     bChildInitial=False
     bParentInitial=False
     Child.Name=@n1.Child
     Parent.Name=@n1.Parent
    ).RelativeTransform.Rotation
   )
   Remap = &RigVMFunction_MathDoubleRemap(
    TargetMaximum=1
    TargetMinimum=0
    Value=@RigVMFunction_MathQuaternionToEuler.Result.Z
    SourceMaximum=@n1.Source Maximum
    SourceMinimum=@n1.Source Minimum
   )
   > RigVMFunctionReturnNode(Delta Rotation=@Remap.Result)
  ]}
 New Function_ContainedGraph {nodes [
   RigVMFunctionEntryNode()
   > n1 = &RigUnit_ParentConstraint(
    AdvancedSettings=(RotationOrderForFilter=XZY)
    Child=(Type=Bone,Name="base_tie")
    Parents="((Item=(Type=Bone,Name=\"spine_02\"),Weight=0.0016),(Item=(Type=Bone,Name=\"clavicle_l\"),Weight=0.0171),(Item=(Type=Bone,Name=\"clavicle_r\"),Weight=0.021),(Item=(Type=Bone,Name=\"neck_01\"),Weight=0.0287),(Item=(Type=Bone,Name=\"spine_03\"),Weight=0.0292),(Item=(T ...(truncated)"
    Weight=1
   )
   RigVMFunction_ControlFlowBranch = &RigVMFunction_ControlFlowBranch(ExecuteContext=@n1.ExecutePin, Condition=RigVMVariableNode() {variable=isFE_CR}.Value)
   RigVMFunctionReturnNode(ExecuteContext=@RigVMFunction_ControlFlowBranch.Completed)
   n2 = &RigUnit_SphericalPoseReader(
    ActiveRegionScaleFactors=(PositiveWidth=1,NegativeWidth=1,PositiveHeight=1,NegativeHeight=1)
    ActiveRegionSize=0.05
    Debug=(bDrawDebug=false,bDraw2D=False,bDrawLocalAxes=False,DebugScale=25,DebugSegments=20,DebugThickness=0.25)
    DriverAxis=(X=1,Y=0,Z=0)
    DriverItem=(Type=Bone,Name="spine_05")
    FalloffRegionScaleFactors=(PositiveWidth=1,NegativeWidth=1,PositiveHeight=1,NegativeHeight=1)
    FalloffSize=0.15
    FlipHeightScaling=False
    FlipWidthScaling=False
    OptionalParentItem=(Type=Bone,Name="spine_01")
    RotationOffset=(X=90,Y=0,Z=-55)
    ExecutePin=@RigVMFunction_ControlFlowBranch.False
   )
   > &RigUnit_ModifyTransforms(
    ItemToModify=((Item=(Type=Bone,Name="base_tie"),Transform=(Rotation=(X=0,Y=0.0773,Z=0,W=0.997),Translation=(X=0,Y=0,Z=0),Scale3D=(X=1,Y=1,Z=1))))
    WeightMinimum=0
    Weight=@n2.OutputParam
   )
   > n3 = &RigUnit_SphericalPoseReader(
    ActiveRegionScaleFactors=(PositiveWidth=1,NegativeWidth=1,PositiveHeight=1,NegativeHeight=1)
    ActiveRegionSize=0.02
    Debug=(bDrawDebug=false,bDraw2D=False,bDrawLocalAxes=False,DebugScale=25,DebugSegments=20,DebugThickness=0.25)
    DriverAxis=(X=1,Y=0,Z=0)
    DriverItem=(Type=Bone,Name="head")
    FalloffRegionScaleFactors=(PositiveWidth=1,NegativeWidth=1,PositiveHeight=1,NegativeHeight=1)
    FalloffSize=0.09
    FlipHeightScaling=False
    FlipWidthScaling=False
    OptionalParentItem=(Type=Bone,Name="spine_05")
    RotationOffset=(X=55,Y=0,Z=-130)
   )
   > &RigUnit_ModifyTransforms(
    ItemToModify=((Item=(Type=Bone,Name="base_tie"),Transform=(Rotation=(X=0.0745,Y=0.0782,Z=-0.0059,W=0.9941),Translation=(X=0,Y=0,Z=0),Scale3D=(X=1,Y=1,Z=1))))
    WeightMinimum=0
    Weight=@n3.OutputParam
   )
   n4 = &RigVMFunction_AlphaInterp(
    InterpSpeedDecreasing=10
    InterpSpeedIncreasing=10
    Scale=1
    bClampResult=true
    Value=&RigVMFunction_MathBoolToFloat(Value=RigVMVariableNode() {variable=isDBNO_CR}.Value).Result
   )
   > &RigUnit_ModifyTransforms(
    ItemToModify=((Item=(Type=Bone,Name="base_tie"),Transform=(Rotation=(X=0,Y=0.5711,Z=0,W=0.8209),Translation=(X=0,Y=0,Z=0),Scale3D=(X=1,Y=1,Z=1))),(Item=(Type=Bone,Name="None"),Transform=(Rotation=(X=0,Y=0,Z=0,W=1),Translation=(X=0,Y=0,Z=0),Scale3D=(X=1,Y=1,Z=1))))
    Weight=@n4.Result
   )
   &RigUnit_ModifyTransforms(
    ItemToModify=((Item=(Type=Bone,Name="base_tie"),Transform=(Rotation=(X=0,Y=0,Z=0,W=1),Translation=(X=0,Y=0,Z=0),Scale3D=(X=1,Y=1,Z=1))))
    Weight=1
    WeightMinimum=0
    ExecutePin=@RigVMFunction_ControlFlowBranch.True
   )
  ]}
 RigVMFunctionLibrary {nodes [
   Delta Rotation Y = RigVMCollapseNode(ExecuteContext=())
   Delta Rotation Z = RigVMCollapseNode(ExecuteContext=())
   Jacket = RigVMCollapseNode(ExecuteContext=())
   Tie = RigVMCollapseNode(ExecuteContext=())
  ]}
 RigVMFunction_Sequence_1_ContainedGraph {nodes [
   n1 = RigVMFunctionEntryNode()
   RigVMFunction_Sequence = &RigVMFunction_Sequence(ExecuteContext=@n1.ExecuteContext)
   n2 = &RigVMFunction_Sequence(ExecuteContext=@RigVMFunction_Sequence.B)
   n3 = &RigVMFunction_Sequence(ExecuteContext=@n2.B)
   n4 = &RigVMFunction_Sequence(ExecuteContext=@n3.B)
   RigVMFunctionReturnNode(A=@RigVMFunction_Sequence.A, B=@n2.A, C=@n3.A, D=@n4.A, E=@n4.B)
  ]}
 RigVMModel {nodes [
   RigUnit_BeginExecution = &RigUnit_BeginExecution()
   FN_DeformRig_FNC = RigVMFunctionReferenceNode(
    DeformRig_ Curve="(deformRig_l_upperarm_up_alpha_72_#1=False,deformRig_r_upperarm_up_alpha_73_#2=False,deformRig_l_upperarm_fr_alpha_74_#3=False,deformRig_r_upperarm_fr_alpha_75_CB51B8 ...(truncated)"
    DeformRig_Bool="(upperarm_1_#4=True,elbow_3_#5=True,wrist_7_#6=True,groin_8_#7=True,glute_17_#8=True,knee_9_40C408324CA261 ...(truncated)"
    DeformRig_Debug=(upperarm_debug_18_#4=False,wrist_debug_20_#6=False,lat_debug_26_#9=False,groinGlute_debug_27_#7=False)
    DeformRig_Value="(upperarm_up_18_#10=(Rotation=(X=0,Y=0.4226,Z=0,W=0.9063),Translation=(X=0.0,Y=0.0,Z=3),Scale3D=(X=0.70,Y=1.0,Z=1.0)),upperarm_fwd_fr_19_#11=(Rotation=(X=0.0,Y=-0.0,Z=0.0,W=1.0),Translation=(X=0.0,Y ...(truncated)"
    ExecuteContext=@RigUnit_BeginExecution.ExecutePin
   )
   > Jacket = RigVMFunctionReferenceNode()
   > Tie = RigVMFunctionReferenceNode()
  ]}
]
hierarchy [
 root {loc=<0 0 0>, rot=<0 0 0 1>}
 pelvis {parent=root, loc=<0 -0.7071 93.1191>, rot=<0.0127 0.707 -0.0127 -0.707>}
 spine_01 {parent=pelvis, loc=<6.7494 -0.6388 0>, rot=<0 0 0.034 -0.9994>}
 spine_02 {parent=spine_01, loc=<7.2518 -0.035 0>, rot=<0 0 0.0238 -0.9997>}
 spine_03 {parent=spine_02, loc=<6.0104 0 0>, rot=<0 0 0.0149 -0.9999>}
 spine_04 {parent=spine_03, loc=<8.1024 0 0>, rot=<0 0 -0.0132 -0.9999>}
 spine_05 {parent=spine_04, loc=<10.0146 0 0>, rot=<0 0 -0.0015 -1>}
 neck_01 {parent=spine_05, loc=<11.3138 0.1529 0>, rot=<0 0 -0.2003 -0.9797>}
 neck_02 {parent=neck_01, loc=<4.5627 0 0>, rot=<0 0 0.0504 -0.9987>}
 head {parent=neck_02, loc=<5.3914 0 0>, rot=<0 0 0.115 -0.9934>}
 attach_cape {parent=spine_05, loc=<9.0186 -11.7097 0.0009>, rot=<0.0616 -0.7044 -0.0616 -0.7044>}
 cape {parent=attach_cape, type=Null, loc=<0 0.5 0>, rot=<0.0698 0 0 0.9976>}
 attach_backpack {parent=spine_05, loc=<-0.1201 -13.4658 0.0005>, rot=<-0.0093 -0.707 0.0093 -0.707>}
 Backpack_BR {parent=attach_backpack, type=Null, loc=<0 0 0>, rot=<0 0 0 1>}
 clavicle_l {parent=spine_05, loc=<4.3575 -0.1466 -2.3223>, rot=<0.0465 -0.7289 -0.009 -0.6829>}
 upperarm_l {parent=clavicle_l, loc=<15.5111 0 0>, rot=<0.0226 -0.3234 -0.0088 -0.9459>}
 lowerarm_l {parent=upperarm_l, loc=<26.6514 0 0>, rot=<0.0391 -0.0099 -0.2453 -0.9686>}
 hand_l {parent=lowerarm_l, loc=<22.3491 0 0>, rot=<-0.7081 -0.0507 0.0975 -0.6975>}
 thumb_01_l {parent=hand_l, loc=<1.3118 -0.3191 -2.459>, rot=<0.6546 -0.4042 0.0664 -0.6354>}
 thumb_02_l {parent=thumb_01_l, loc=<5.9336 -0.0448 0.0401>, rot=<-0.0037 0.0213 0.298 -0.9543>}
 thumb_03_l {parent=thumb_02_l, loc=<2.2223 0.3682 -0.0907>, rot=<0.0118 -0.013 0.0066 -0.9998>}
 index_metacarpal_l {parent=hand_l, loc=<1.6535 0.0183 -2.96>, rot=<0.0158 -0.0124 -0.0173 -0.9996>}
 index_01_l {parent=index_metacarpal_l, loc=<8.3008 0.0357 -0.0524>, rot=<-0.0044 0.0485 0.0924 -0.9945>}
 index_02_l {parent=index_01_l, loc=<3.3587 -0.3004 0.0452>, rot=<-0.002 0.0027 0.3225 -0.9466>}
 index_03_l {parent=index_02_l, loc=<2.5578 -0.123 -0.0252>, rot=<-0.002 0.0067 0.2558 -0.9667>}
 middle_metacarpal_l {parent=hand_l, loc=<2.1856 -0.1725 -0.5984>, rot=<0.006 0.0168 -0.0329 -0.9993>}
 middle_01_l {parent=middle_metacarpal_l, loc=<7.3275 -0.2447 0.0129>, rot=<-0.0208 0.0042 0.0858 -0.9961>}
 middle_02_l {parent=middle_01_l, loc=<4.0123 0.1401 0.1175>, rot=<-0.0218 0.0258 0.3512 -0.9357>}
 middle_03_l {parent=middle_02_l, loc=<2.6077 0.0534 0.1373>, rot=<0.0082 0.0188 0.1708 -0.9851>}
 ring_metacarpal_l {parent=hand_l, loc=<2.2074 -0.3625 1.3638>, rot=<-0.0284 0.0722 -0.0458 -0.9959>}
 ring_01_l {parent=ring_metacarpal_l, loc=<7.0295 -0.6197 -0.1446>, rot=<-0.0489 -0.0252 0.1892 -0.9804>}
 ring_02_l {parent=ring_01_l, loc=<3.6174 0.4647 -0.1027>, rot=<0 0 0.2806 -0.9598>}
 ring_03_l {parent=ring_02_l, loc=<2.4152 0.114 0.0199>, rot=<0 0 0.1189 -0.9929>}
 pinky_metacarpal_l {parent=hand_l, loc=<1.7616 -0.9753 3.3814>, rot=<-0.0397 0.0814 -0.069 -0.9935>}
 pinky_01_l {parent=pinky_metacarpal_l, loc=<7.1709 -0.4831 -0.1094>, rot=<-0.0654 -0.0035 0.2242 -0.9723>}
 pinky_02_l {parent=pinky_01_l, loc=<2.8274 0.0893 -0.1043>, rot=<0.0037 0.0011 0.2854 -0.9584>}
 pinky_03_l {parent=pinky_02_l, loc=<2.2046 0.0307 -0.0078>, rot=<-0.0002 0.0098 0.1445 -0.9895>}
 deform_wrist_dn_base_l {parent=hand_l, loc=<0 -2.2342 0>, rot=<0 0 0 -1>}
 deform_wrist_dn_l {parent=deform_wrist_dn_base_l, loc=<0 0 0>, rot=<0 0 0 1>}
 deform_wrist_up_base_l {parent=hand_l, loc=<0 2.2682 0>, rot=<0 0 0 -1>}
 deform_wrist_up_l {parent=deform_wrist_up_base_l, loc=<0 0 0>, rot=<0 0 0 1>}
 lowerarm_twist_02_l {parent=lowerarm_l, loc=<9.6199 0 0>, rot=<0 0 0 -1>}
 lowerarm_twist_01_l {parent=lowerarm_l, loc=<17.8549 0 0>, rot=<0 0 0 -1>}
 deform_elbow_in_base_l {parent=lowerarm_l, loc=<0.2112 5.259 0.8274>, rot=<0 0 0 -1>}
 deform_elbow_in_l {parent=deform_elbow_in_base_l, loc=<0 0 0>, rot=<0 0 0 1>}
 deform_elbow_out_base_l {parent=lowerarm_l, loc=<0 -4 0>, rot=<0 0 0 -1>}
 deform_elbow_out_l {parent=deform_elbow_out_base_l, loc=<0 0 0>, rot=<0 0 0 1>}
 upperarm_twist_01_l {parent=upperarm_l, loc=<0.0041 0 0>, rot=<0 0 0 -1>}
 deform_upperarm_bk_base_l {parent=upperarm_twist_01_l, loc=<1.0813 -7.2193 -1.9453>, rot=<0 0 0 1>}
 deform_upperarm_bk_l {parent=deform_upperarm_bk_base_l, loc=<0 0 0>, rot=<0 0 0 1>}
 deform_upperarm_fr_base_l {parent=upperarm_twist_01_l, loc=<-0.3691 7.0506 -2.765>, rot=<0 0 0 1>}
 deform_upperarm_fr_l {parent=deform_upperarm_fr_base_l, loc=<0 0 0>, rot=<0 0 0 1>}
 deform_upperarm_up_base_l {parent=upperarm_twist_01_l, loc=<-3.2216 -1.1391 3.3311>, rot=<0 0 0 1>}
 deform_upperarm_up_l {parent=deform_upperarm_up_base_l, loc=<0 0 0>, rot=<0 0 0 1>}
 upperarm_twist_02_l {parent=upperarm_l, loc=<13.7941 0 0>, rot=<0 0 0 -1>}
 clavicle_r {parent=spine_05, loc=<4.3575 -0.1466 2.3223>, rot=<0.7289 0.0465 -0.6829 0.009>}
 upperarm_r {parent=clavicle_r, loc=<-15.5111 0 0>, rot=<0.0226 -0.3234 -0.0088 -0.9459>}
 lowerarm_r {parent=upperarm_r, loc=<-26.6514 0 0>, rot=<0.0391 -0.0099 -0.2453 -0.9686>}
 hand_r {parent=lowerarm_r, loc=<-22.3491 0 0>, rot=<-0.7081 -0.0507 0.0975 -0.6975>}
 thumb_01_r {parent=hand_r, loc=<-1.3118 0.3191 2.459>, rot=<0.6546 -0.4042 0.0664 -0.6354>}
 thumb_02_r {parent=thumb_01_r, loc=<-5.9336 0.0448 -0.0401>, rot=<-0.0037 0.0213 0.298 -0.9543>}
 thumb_03_r {parent=thumb_02_r, loc=<-2.2223 -0.3682 0.0907>, rot=<0.0118 -0.013 0.0066 -0.9998>}
 index_metacarpal_r {parent=hand_r, loc=<-1.6535 -0.0183 2.96>, rot=<0.0158 -0.0124 -0.0173 -0.9996>}
 index_01_r {parent=index_metacarpal_r, loc=<-8.3008 -0.0357 0.0524>, rot=<-0.0044 0.0485 0.0924 -0.9945>}
 index_02_r {parent=index_01_r, loc=<-3.3587 0.3004 -0.0452>, rot=<-0.002 0.0027 0.3225 -0.9466>}
 index_03_r {parent=index_02_r, loc=<-2.5578 0.123 0.0252>, rot=<-0.002 0.0067 0.2558 -0.9667>}
 middle_metacarpal_r {parent=hand_r, loc=<-2.1856 0.1725 0.5984>, rot=<0.006 0.0168 -0.0329 -0.9993>}
 middle_01_r {parent=middle_metacarpal_r, loc=<-7.3275 0.2447 -0.0129>, rot=<-0.0208 0.0042 0.0858 -0.9961>}
 middle_02_r {parent=middle_01_r, loc=<-4.0123 -0.1401 -0.1175>, rot=<-0.0218 0.0258 0.3512 -0.9357>}
 middle_03_r {parent=middle_02_r, loc=<-2.6077 -0.0534 -0.1373>, rot=<0.0082 0.0188 0.1708 -0.9851>}
 ring_metacarpal_r {parent=hand_r, loc=<-2.2074 0.3625 -1.3638>, rot=<-0.0284 0.0722 -0.0458 -0.9959>}
 ring_01_r {parent=ring_metacarpal_r, loc=<-7.0295 0.6197 0.1446>, rot=<-0.0489 -0.0252 0.1892 -0.9804>}
 ring_02_r {parent=ring_01_r, loc=<-3.6174 -0.4647 0.1027>, rot=<0 0 0.2806 -0.9598>}
 ring_03_r {parent=ring_02_r, loc=<-2.4152 -0.114 -0.0199>, rot=<0 0 0.1189 -0.9929>}
 pinky_metacarpal_r {parent=hand_r, loc=<-1.7616 0.9753 -3.3814>, rot=<-0.0397 0.0814 -0.069 -0.9935>}
 pinky_01_r {parent=pinky_metacarpal_r, loc=<-7.1709 0.4831 0.1094>, rot=<-0.0654 -0.0035 0.2242 -0.9723>}
 pinky_02_r {parent=pinky_01_r, loc=<-2.8274 -0.0893 0.1043>, rot=<0.0037 0.0011 0.2854 -0.9584>}
 pinky_03_r {parent=pinky_02_r, loc=<-2.2046 -0.0307 0.0078>, rot=<-0.0002 0.0098 0.1445 -0.9895>}
 deform_wrist_dn_base_r {parent=hand_r, loc=<-0.0004 2.2345 -0.0001>, rot=<0 0 0 -1>}
 deform_wrist_dn_r {parent=deform_wrist_dn_base_r, loc=<0 0 0>, rot=<0 0 0 1>}
 deform_wrist_up_base_r {parent=hand_r, loc=<0.0003 -2.2684 0.0001>, rot=<0 0 0 -1>}
 deform_wrist_up_r {parent=deform_wrist_up_base_r, loc=<0 0 0>, rot=<0 0 0 1>}
 lowerarm_twist_02_r {parent=lowerarm_r, loc=<-9.6199 0 0>, rot=<0 0 0 -1>}
 lowerarm_twist_01_r {parent=lowerarm_r, loc=<-17.8549 0 0>, rot=<0 0 0 -1>}
 deform_elbow_in_base_r {parent=lowerarm_r, loc=<-0.2111 -5.2591 -0.8275>, rot=<0 0 0 -1>}
 deform_elbow_in_r {parent=deform_elbow_in_base_r, loc=<-0.0001 0 0.0002>, rot=<0 0 0 1>}
 deform_elbow_out_base_r {parent=lowerarm_r, loc=<0 4 -0.0001>, rot=<0 0 0 -1>}
 deform_elbow_out_r {parent=deform_elbow_out_base_r, loc=<0 0 0>, rot=<0 0 0 1>}
 upperarm_twist_01_r {parent=upperarm_r, loc=<-0.0041 0 0>, rot=<0 0 0 -1>}
 deform_upperarm_bk_base_r {parent=upperarm_twist_01_r, loc=<-1.0814 7.2192 1.9455>, rot=<0 0 0 1>}
 deform_upperarm_bk_r {parent=deform_upperarm_bk_base_r, loc=<0 0 0>, rot=<0 0 0 1>}
 deform_upperarm_fr_base_r {parent=upperarm_twist_01_r, loc=<0.3689 -7.0506 2.7652>, rot=<0 0 0 1>}
 deform_upperarm_fr_r {parent=deform_upperarm_fr_base_r, loc=<0 0 0>, rot=<0 0 0 1>}
 deform_upperarm_up_base_r {parent=upperarm_twist_01_r, loc=<3.2214 1.1391 -3.3308>, rot=<0 0 0 1>}
 deform_upperarm_up_r {parent=deform_upperarm_up_base_r, loc=<0 0 0>, rot=<0 0 0 -1>}
 upperarm_twist_02_r {parent=upperarm_r, loc=<-13.7941 0 0>, rot=<0 0 0 -1>}
 base_tie {parent=spine_05, loc=<6.1941 9.0351 0>, rot=<-0.1178 -0.6972 0.6972 0.1178>}
 dyn_tie_01 {parent=base_tie, loc=<0 0 0>, rot=<0 0 0 1>}
 dyn_tie_02 {parent=dyn_tie_01, loc=<7.1291 0 0>, rot=<0 0.1122 0 -0.9937>}
 dyn_tie_03 {parent=dyn_tie_02, loc=<7.1291 0 0>, rot=<0 0.0518 0 -0.9987>}
 dyn_tie_04 {parent=dyn_tie_03, loc=<7.1291 0 0>, rot=<0 0 0 1>}
 dyn_tie_05 {parent=dyn_tie_04, loc=<7.1291 0 0>, rot=<0 0 0 1>}
 tie_knot_mstr {parent=spine_05, loc=<8.4242 7.9283 -0.0009>, rot=<0.1135 0.6979 0.6979 0.1135>}
 tie_knot_01_dyn {parent=tie_knot_mstr, loc=<0 0 0>, rot=<0 0 0 1>}
 tie_knot_02_dyn {parent=tie_knot_01_dyn, loc=<2.3387 -0.0272 -0.0507>, rot=<-0.0501 0.0438 0.024 -0.9975>}
 tie_01_dyn {parent=tie_knot_02_dyn, loc=<7.9567 0 0>, rot=<-0.0007 -0.1625 0.0003 -0.9867>}
 tie_02_dyn {parent=tie_01_dyn, loc=<12.6996 -0.0269 0.1565>, rot=<0.034 -0.0715 -0.0032 -0.9969>}
 tie_03_dyn {parent=tie_02_dyn, loc=<10.4555 -0.0115 0.2595>, rot=<-0.0381 0.0001 0.0008 -0.9993>}
 deform_pec_aim_base_l {parent=spine_04, loc=<7.8732 12.6388 -2.4123>, rot=<0.0102 -0.4602 0.3979 -0.7936>}
 deform_pec_aim_l {parent=deform_pec_aim_base_l, loc=<0 0 0>, rot=<0 0 0 1>}
 deform_pec_base_l {parent=deform_pec_aim_l, loc=<10.766 0 0>, rot=<0 0 0 1>}
 deform_pec_l {parent=deform_pec_base_l, loc=<0.055 0 0>, rot=<0 0 0 1>}
 deform_pec_aim_base_r {parent=spine_04, loc=<7.8727 12.6388 2.4123>, rot=<-0.4602 -0.0102 0.7936 0.3979>}
 deform_pec_aim_r {parent=deform_pec_aim_base_r, loc=<0 0 0>, rot=<0 0 0 1>}
 deform_pec_base_r {parent=deform_pec_aim_r, loc=<-10.7662 -0.0006 -0.0007>, rot=<0 0 0 1>}
 deform_pec_r {parent=deform_pec_base_r, loc=<0 0 0>, rot=<0 0 0 1>}
 thigh_l {parent=pelvis, loc=<-2.5693 -0.0109 -7.7501>, rot=<0.0012 0.0572 0.0155 -0.9982>}
 calf_l {parent=thigh_l, loc=<-39.0484 0 0>, rot=<0.0006 -0.0161 -0.0391 -0.9991>}
 foot_l {parent=calf_l, loc=<-42.561 0.1077 0.4292>, rot=<0.0418 -0.0455 0.0343 -0.9975>}
 ball_l {parent=foot_l, loc=<-6.1599 13.8264 0.6806>, rot=<0.0001 -0.0001 -0.7063 -0.7079>}
 calf_twist_02_l {parent=calf_l, loc=<-18.8064 -0.0967 0.0397>, rot=<0.0237 0.0095 -0.0035 -0.9997>}
 calf_twist_01_l {parent=calf_l, loc=<-35.813 -0.0967 0.0397>, rot=<0.0237 0.0095 -0.0035 -0.9997>}
 deform_knee_in_base_l {parent=calf_l, loc=<1.298 -5.2957 -0.0496>, rot=<0 0 0 -1>}
 deform_knee_in_l {parent=deform_knee_in_base_l, loc=<0 0 0>, rot=<0 0 0 1>}
 deform_knee_out_base_l {parent=calf_l, loc=<0 4 0>, rot=<0 0 0 -1>}
 deform_knee_out_l {parent=deform_knee_out_base_l, loc=<0 0 0>, rot=<0 0 0 1>}
 thigh_twist_01_l {parent=thigh_l, loc=<-21.6303 0 0>, rot=<0.0013 0 0 -1>}
 deform_groin_base_l {parent=thigh_l, loc=<-2.1908 8.1142 -1.7442>, rot=<0 0 0 -1>}
 deform_groin_l {parent=deform_groin_base_l, loc=<0 0 0>, rot=<0 0 0 1>}
 thigh_r {parent=pelvis, loc=<-2.5693 -0.0109 7.7501>, rot=<0.0572 -0.0012 0.9982 0.0155>}
 calf_r {parent=thigh_r, loc=<39.0484 0 0>, rot=<0.0006 -0.0161 -0.0391 -0.9991>}
 foot_r {parent=calf_r, loc=<42.561 -0.1077 -0.4292>, rot=<0.0418 -0.0455 0.0343 -0.9975>}
 ball_r {parent=foot_r, loc=<6.1599 -13.8264 -0.6806>, rot=<0.0001 -0.0001 -0.7063 -0.7079>}
 calf_twist_02_r {parent=calf_r, loc=<18.8064 0.0967 -0.0397>, rot=<0.0237 0.0095 -0.0035 -0.9997>}
 calf_twist_01_r {parent=calf_r, loc=<35.813 0.0967 -0.0397>, rot=<0.0237 0.0095 -0.0035 -0.9997>}
 deform_knee_in_base_r {parent=calf_r, loc=<-1.298 5.2957 0.0496>, rot=<0 0 0 -1>}
 deform_knee_in_r {parent=deform_knee_in_base_r, loc=<0 0 0>, rot=<0 0 0 1>}
 deform_knee_out_base_r {parent=calf_r, loc=<0 -4 0>, rot=<0 0 0 -1>}
 deform_knee_out_r {parent=deform_knee_out_base_r, loc=<0 0 0>, rot=<0 0 0 1>}
 thigh_twist_01_r {parent=thigh_r, loc=<21.6303 0 0>, rot=<0.0013 0 0 -1>}
 deform_groin_base_r {parent=thigh_r, loc=<2.1908 -8.1142 1.7442>, rot=<0 0 0 -1>}
 deform_groin_r {parent=deform_groin_base_r, loc=<0 0 0>, rot=<0 0 0 1>}
 deform_glute_base_l {parent=pelvis, loc=<-8.8064 -8.8222 -8.4287>, rot=<0.0275 0.0501 -0.4472 -0.8926>}
 deform_glute_l {parent=deform_glute_base_l, loc=<0 0 0>, rot=<0 0 0 1>}
 deform_glute_base_r {parent=pelvis, loc=<-8.8062 -8.8222 8.4287>, rot=<0.0501 -0.0275 0.8926 -0.4472>}
 deform_glute_r {parent=deform_glute_base_r, loc=<0 0 0>, rot=<0 0 0 1>}
 base_suitSideFront_l {parent=pelvis, loc=<13.502 7.9047 -9.0345>, rot=<-0.0178 0.2574 0.9638 0.0667>}
 dyn_suitSideFront_01_l {parent=base_suitSideFront_l, loc=<0 0 0>, rot=<0 0 0 1>}
 dyn_suitSideFront_02_l {parent=dyn_suitSideFront_01_l, loc=<10.4031 0 0>, rot=<0 0 0 1>}
 base_suitSide_l {parent=pelvis, loc=<13.5024 -0.1646 -15.5539>, rot=<-0.045 0.6521 0.755 0.0521>}
 dyn_suitSide_01_l {parent=base_suitSide_l, loc=<0 0 0>, rot=<0 0 0 1>}
 dyn_suitSide_02_l {parent=dyn_suitSide_01_l, loc=<10.4031 0 0>, rot=<0 0 0 1>}
 base_suitSideBack_l {parent=pelvis, loc=<13.5363 -10.1988 -10.7507>, rot=<0.0362 -0.959 -0.2809 -0.0106>}
 dyn_suitSideBack_01_l {parent=base_suitSideBack_l, loc=<0 0 0>, rot=<0 0 0 1>}
 dyn_suitSideBack_02_l {parent=dyn_suitSideBack_01_l, loc=<10.4031 0 0>, rot=<0 0 0 1>}
 base_suitBack {parent=pelvis, loc=<13.5476 -12.8277 0>, rot=<0.0178 -0.9998 0 0>}
 dyn_suitBack_01 {parent=base_suitBack, loc=<0 0 0>, rot=<0 0 0 1>}
 dyn_suitBack_02 {parent=dyn_suitBack_01, loc=<10.4031 0 0>, rot=<0 0 0 1>}
 base_suitSideFront_r {parent=pelvis, loc=<13.502 7.9047 9.0346>, rot=<0.0178 -0.2574 0.9638 0.0667>}
 dyn_suitSideFront_01_r {parent=base_suitSideFront_r, loc=<0 0 0>, rot=<0 0 0 1>}
 dyn_suitSideFront_02_r {parent=dyn_suitSideFront_01_r, loc=<10.4031 0 0>, rot=<0 0 0 1>}
 base_suitSide_r {parent=pelvis, loc=<13.5024 -0.1646 15.5542>, rot=<0.045 -0.6521 0.755 0.0521>}
 dyn_suitSide_01_r {parent=base_suitSide_r, loc=<0 0 0>, rot=<0 0 0 1>}
 dyn_suitSide_02_r {parent=dyn_suitSide_01_r, loc=<10.4031 0 0>, rot=<0 0 0 1>}
 base_suitSideBack_r {parent=pelvis, loc=<13.5363 -10.1988 10.7503>, rot=<0.0362 -0.959 0.2809 0.0106>}
 dyn_suitSideBack_01_r {parent=base_suitSideBack_r, loc=<0 0 0>, rot=<0 0 0 1>}
 dyn_suitSideBack_02_r {parent=dyn_suitSideBack_01_r, loc=<10.4031 0 0>, rot=<0 0 0 1>}
 thigh_r_null {parent=pelvis, type=Null, loc=<-2.5693 -0.0109 7.7501>, rot=<0.0572 -0.0012 0.9982 0.0155>}
 thigh_l_null {parent=pelvis, type=Null, loc=<-2.5693 -0.0109 -7.7501>, rot=<0.0012 0.0572 0.0155 -0.9982>}
 dyn_jacket_fr_l_01_MSTR {parent=pelvis, loc=<11.623 9.3411 -7.3302>, rot=<-0.031 -0.8398 -0.5409 -0.0344>}
 dyn_jacket_fr_l_01 {parent=dyn_jacket_fr_l_01_MSTR, loc=<0 0 0>, rot=<0 0 0 1>}
 dyn_jacket_fr_l_02 {parent=dyn_jacket_fr_l_01, loc=<11.6956 0 0>, rot=<-0.0543 0.0177 0.0542 -0.9969>}
 dyn_jacket_l_01_MSTR {parent=pelvis, loc=<11.22 3.1397 -14.8441>, rot=<-0.0389 -0.9788 -0.1955 -0.0467>}
 dyn_jacket_l_01 {parent=dyn_jacket_l_01_MSTR, loc=<0 0 0>, rot=<0 0 0 1>}
 dyn_jacket_l_02 {parent=dyn_jacket_l_01, loc=<9.987 0 0>, rot=<0.0198 0.0152 0.0224 -0.9994>}
 dyn_jacket_bkout_l_01_MSTR {parent=pelvis, loc=<11.2458 -3.3873 -16.5405>, rot=<-0.0146 -0.9964 0.0707 -0.0445>}
 dyn_jacket_bkout_l_01 {parent=dyn_jacket_bkout_l_01_MSTR, loc=<0 0 0>, rot=<0 0 0 1>}
 dyn_jacket_bkout_l_02 {parent=dyn_jacket_bkout_l_01, loc=<9.9736 0 0>, rot=<-0.0328 0.0181 -0.0227 -0.999>}
 dyn_jacket_bk_l_01_MSTR {parent=pelvis, loc=<11.0236 -9.2268 -13.4921>, rot=<-0.0004 -0.942 0.333 -0.0408>}
 dyn_jacket_bk_l_01 {parent=dyn_jacket_bk_l_01_MSTR, loc=<0 0 0>, rot=<0 0 0 1>}
 dyn_jacket_bk_l_02 {parent=dyn_jacket_bk_l_01, loc=<10.021 0 0>, rot=<-0.018 -0.0085 0.0127 -0.9997>}
 dyn_jacket_bk_cr_l_01_MSTR {parent=pelvis, loc=<10.9506 -12.8199 -3.909>, rot=<0.034 -0.7154 0.6977 -0.0153>}
 dyn_jacket_bk_cr_l_01 {parent=dyn_jacket_bk_cr_l_01_MSTR, loc=<0 0 0>, rot=<0 0 0 1>}
 dyn_jacket_bk_cr_l_02 {parent=dyn_jacket_bk_cr_l_01, loc=<10.7431 0 0>, rot=<-0.008 0.057 -0.013 -0.9983>}
 dyn_jacket_bk_cr_r_01_MSTR {parent=pelvis, loc=<10.8439 -12.6802 3.763>, rot=<0.0168 -0.7118 0.702 -0.0174>}
 dyn_jacket_bk_cr_r_01 {parent=dyn_jacket_bk_cr_r_01_MSTR, loc=<0 0 0>, rot=<0 0 0 1>}
 dyn_jacket_bk_cr_r_02 {parent=dyn_jacket_bk_cr_r_01, loc=<10.6609 0 0>, rot=<-0.0012 0.0675 0.0002 -0.9977>}
 dyn_jacket_bk_r_01_MSTR {parent=pelvis, loc=<10.8779 -9.2603 13.5361>, rot=<0.0384 -0.3703 0.9281 -0.0013>}
 dyn_jacket_bk_r_01 {parent=dyn_jacket_bk_r_01_MSTR, loc=<0 0 0>, rot=<0 0 0 1>}
 dyn_jacket_bk_r_02 {parent=dyn_jacket_bk_r_01, loc=<10.09 0 0>, rot=<0.0271 -0.0115 -0.0023 -0.9996>}
 dyn_jacket_bkout_r_01_MSTR {parent=pelvis, loc=<10.9177 -3.8834 16.2735>, rot=<0.0465 -0.1518 0.9859 0.0538>}
 dyn_jacket_bkout_r_01 {parent=dyn_jacket_bkout_r_01_MSTR, loc=<0 0 0>, rot=<0 0 0 1>}
 dyn_jacket_bkout_r_02 {parent=dyn_jacket_bkout_r_01, loc=<9.8271 0 0>, rot=<-0.0724 0.0163 -0.0183 -0.9971>}
 dyn_jacket_r_01_MSTR {parent=pelvis, loc=<10.8095 6.1599 12.7427>, rot=<0.0503 0.3831 0.9223 0.0109>}
 dyn_jacket_r_01 {parent=dyn_jacket_r_01_MSTR, loc=<0 0 0>, rot=<0 0 0 1>}
 dyn_jacket_r_02 {parent=dyn_jacket_r_01, loc=<9.871 0 0>, rot=<0.0032 0.0044 0.0103 -0.9999>}
 dyn_jacket_fr_r_01_MSTR {parent=pelvis, loc=<10.6739 9.9465 7.28>, rot=<0.0383 0.4312 0.9012 0.0219>}
 dyn_jacket_fr_r_01 {parent=dyn_jacket_fr_r_01_MSTR, loc=<0 0 0>, rot=<0 0 0 1>}
 dyn_jacket_fr_r_02 {parent=dyn_jacket_fr_r_01, loc=<10.3341 0 0>, rot=<-0.0634 0.0221 -0.0355 -0.9971>}
 base_suit_button_01 {parent=pelvis, loc=<13.471 9.6961 6.753>, rot=<-0.1269 0.1185 -0.6457 -0.7435>}
 dyn_suit_button_01 {parent=base_suit_button_01, loc=<0 0 0>, rot=<0 0 0 1>}
 base_suit_button_02 {parent=pelvis, loc=<4.4655 10.7909 7.0217>, rot=<-0.167 0.163 -0.6436 -0.7289>}
 dyn_suit_button_02 {parent=base_suit_button_02, loc=<0 0 0>, rot=<0 0 0 1>}
 base_suit_button_02_LookAt {parent=pelvis, loc=<1.7275 11.1104 7.1147>, rot=<-0.167 0.163 -0.6436 -0.7289>}
 dyn_suit_button_02_LookAt {parent=base_suit_button_02_LookAt, loc=<0 0 0>, rot=<0 0 0 1>, scale=<0.9731 1 1>}
 base_suit_button_01_LookAt {parent=pelvis, loc=<12.1656 9.8725 6.8001>, rot=<-0.1269 0.1185 -0.6457 -0.7435>}
 dyn_suit_button_01_LookAt {parent=base_suit_button_01_LookAt, loc=<0 0 0>, rot=<0 0 0 1>}
 base_pelvis_fix {parent=pelvis, loc=<16.8328 0.1042 0>, rot=<-0.0127 -0.707 0.0127 -0.707>, scale=<0.8854 0.8854 0.8854>}
 thigh_r_ctrl {controlType=EULER_TRANSFORM, type=Control, loc=<0 0 0>, rot=<0 0 0 1>}
 thigh_l_ctrl {controlType=EULER_TRANSFORM, type=Control, loc=<0 0 0>, rot=<0 0 0 1>}
 spine_01_ctrl {controlType=EULER_TRANSFORM, type=Control, loc=<0 0 0>, rot=<0 0 0 1>}
]
variables [
 Thigh_L_SideRotation {type=double, default=0}
 Thigh_R_SideRotation {type=double, default=0}
 isBaseballSliding_CR {type=bool, default=False}
 isCrouching_CR {type=bool, default=False}
 isDBNO_CR {type=bool, default=False}
 isFE_CR {type=bool, default=False}
 isKneeSliding_CR {type=bool, default=False}
]
