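# Physics asset for the jacket of the medium male YokeQuick cosmetic body, simulating its lower flaps around the hips and legs.
# Seven suit base bones (back, side-back, side and side-front) are locked to the pelvis, each anchoring a two-link dynamic chain.
# Chain links have locked twist and limited swing, with different limits on the front and right-side flaps.
# Capsules on pelvis, glutes, thighs, calves and feet form the body volume, with doubled capsules per thigh.
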PhysicsAsset /BRCosmetics/Characters/Player/Male/Medium/Bodies/M_MED_YokeQuick/Meshes/M_MED_YokeQuick_Jacket_PhysicsAsset
bodies {
 base_suitBack [{shapeType=Capsule, length=0.505, radius=0.505, shapeName=base_suitBack_capsule, rotation=<rot 0 -90 0>}]
 base_suitSideBack_l [{shapeType=Capsule, length=0.505, radius=0.505, shapeName=base_suitSideBack_l_capsule, rotation=<rot 0 -90 0>}]
 base_suitSideBack_r [{shapeType=Capsule, length=0.505, radius=0.505, shapeName=base_suitSideBack_r_capsule, rotation=<rot 0 -90 0>}]
 base_suitSideFront_l [{shapeType=Capsule, length=0.505, radius=0.505, shapeName=base_suitSideFront_l_capsule, rotation=<rot 0 -90 0>}]
 base_suitSideFront_r [{shapeType=Capsule, length=0.505, radius=0.505, shapeName=base_suitSideFront_r_capsule, rotation=<rot 0 -90 0>}]
 base_suitSide_l [{shapeType=Capsule, length=0.505, radius=0.505, shapeName=base_suitSide_l_capsule, rotation=<rot 0 -90 0>}]
 base_suitSide_r [{shapeType=Capsule, length=0.505, radius=0.505, shapeName=base_suitSide_r_capsule, rotation=<rot 0 -90 0>}]
 calf_l [{shapeType=Capsule, length=25, radius=7, shapeName=calf_l_capsule, center=<-25 -1.7 0.5>, rotation=<rot 90 -5.8227 0>}]
 calf_r [{shapeType=Capsule, length=25, radius=7, shapeName=calf_r_capsule, center=<26 1.7 -0.5>, rotation=<rot 90 -5.8227 0>}]
 deform_glute_l [{
   shapeType=Capsule
   length=3.4775
   radius=4.273
   shapeName=deform_glute_l_capsule
   center=<4.0604 -3.7584 -2.1788>
   rotation=<rot 0 -90.0001 -30.7074>
  }]
 deform_glute_r [{
   shapeType=Capsule
   length=3.4775
   radius=4.273
   shapeName=deform_glute_r_capsule
   center=<-4.0604 3.7584 2.1788>
   rotation=<rot 0 -90.0001 149.2926>
  }]
 dyn_suitBack_01 [{
   shapeType=Capsule
   length=13.3981
   radius=7.2622
   shapeName=dyn_suitSideFront_02_l_capsule
   center=<9.8084 -7.1596 0.1099>
   rotation=<rot -7.3435 -90.0005 -0.0003>
  }]
 dyn_suitBack_02 [{
   shapeType=Capsule
   length=13.3981
   radius=7.2622
   shapeName=dyn_suitSideFront_02_l_capsule
   center=<14.2569 -5.8512 0>
   rotation=<rot 0 -90 0>
  }]
 dyn_suitSideBack_01_l [{
   shapeType=Capsule
   length=13.3981
   radius=7.2622
   shapeName=dyn_suitSideFront_02_l_capsule
   center=<9.8084 -6.5131 0.1932>
   rotation=<rot -7.3435 -90.0005 -0.0003>
  }]
 dyn_suitSideBack_01_r [{
   shapeType=Capsule
   length=13.3981
   radius=7.2622
   shapeName=dyn_suitSideFront_02_l_capsule
   center=<10.6609 -7.3038 0.0913>
   rotation=<rot -7.3435 -90.0005 -0.0003>
  }]
 dyn_suitSideBack_02_l [{
   shapeType=Capsule
   length=13.3981
   radius=7.2622
   shapeName=dyn_suitSideFront_02_l_capsule
   center=<14.2567 -5.8379 0.0164>
   rotation=<rot -1.138 -89.8395 -0.9259>
  }]
 dyn_suitSideBack_02_r [{
   shapeType=Capsule
   length=13.3981
   radius=7.2622
   shapeName=dyn_suitSideFront_02_l_capsule
   center=<14.2563 -5.7105 0.0257>
   rotation=<rot -1.138 -89.8395 -0.9259>
  }]
 dyn_suitSideFront_01_l [{
   shapeType=Capsule
   length=13.5184
   radius=6.548
   shapeName=dyn_suitSideFront_02_l_capsule
   center=<9.8084 -5.7977 -0.0061>
   rotation=<rot 5.7016 -90.0009 -0.0005>
  }]
 dyn_suitSideFront_01_r [{
   shapeType=Capsule
   length=13.3981
   radius=7.2622
   shapeName=dyn_suitSideFront_02_l_capsule
   center=<9.8083 -8.4041 0.0572>
   rotation=<rot -0.1629 -90.0008 -0.0001>
  }]
 dyn_suitSideFront_02_l [{
   shapeType=Capsule
   length=13.6284
   radius=6.2778
   shapeName=dyn_suitSideFront_02_l_capsule
   center=<14.2565 -6.4353 0.1268>
   rotation=<rot 10.7552 -90.0349 -0.9424>
  }]
 dyn_suitSideFront_02_r [{
   shapeType=Capsule
   length=13.3981
   radius=6.5081
   shapeName=dyn_suitSideFront_02_l_capsule
   center=<14.2555 -6.8736 0.2584>
   rotation=<rot -3.565 -89.8009 -0.9272>
  }]
 dyn_suitSide_01_l [{
   shapeType=Capsule
   length=13.3981
   radius=7.2622
   shapeName=dyn_suitSideFront_02_l_capsule
   center=<9.8084 -6.1656 0.238>
   rotation=<rot -7.3435 -90.0005 -0.0003>
  }]
 dyn_suitSide_01_r [{
   shapeType=Capsule
   length=13.3981
   radius=7.2622
   shapeName=dyn_suitSideFront_02_l_capsule
   center=<8.1383 -6.6282 0>
   rotation=<rot 0 -90 0>
  }]
 dyn_suitSide_02_l [{
   shapeType=Capsule
   length=13.3981
   radius=7.2622
   shapeName=dyn_suitSideFront_02_l_capsule
   center=<14.2539 -5.1928 0>
   rotation=<rot -6.6548 -89.7504 -0.9321>
  }]
 dyn_suitSide_02_r [{
   shapeType=Capsule
   length=13.3981
   radius=7.2622
   shapeName=dyn_suitSideFront_02_l_capsule
   center=<14.2553 -5.4018 0.0406>
   rotation=<rot -2.2874 -89.8215 -0.9265>
  }]
 foot_l [{
   shapeType=Capsule
   length=9.8449
   radius=8.5179
   shapeName=foot_l_capsule
   center=<-4.2407 4.1563 0.5157>
   rotation=<rot -1.3977 -175.9875 -87.7512>
  }]
 foot_r [{
   shapeType=Capsule
   length=9.8395
   radius=8.5066
   shapeName=foot_r_capsule
   center=<4.2118 -4.1732 -0.4557>
   rotation=<rot -1.3977 -175.7104 -87.8281>
  }]
 pelvis [{shapeType=Capsule, length=11.274, radius=9.9763, center=<3.3821 -0.5468 0.2718>, rotation=<rot 4.4575 -90.0015 177.6591>}]
 thigh_l [
  {shapeType=Capsule, length=29.9507, radius=7.579, center=<-16.8369 -1.8512 0.0429>, rotation=<rot 0.2482 -92.6907 -89.0955>}
  {shapeType=Capsule, length=30.9743, radius=8.2045, center=<-16.9984 0.986 -0.2397>, rotation=<rot 0.1349 -85.5737 -89.0716>}
 ]
 thigh_r [
  {shapeType=Capsule, length=29.9507, radius=7.2741, center=<16.8369 1.8512 -0.0429>, rotation=<rot 0.2482 -92.6907 90.9045>}
  {shapeType=Capsule, length=30.0849, radius=8.4579, center=<16.9415 -2.3075 -0.0933>, rotation=<rot 0.1718 -87.9117 90.922>}
 ]
}
constraints [
 {bone1Name=thigh_l, bone2Name=pelvis, swing1LimitDegrees=45, swing2LimitDegrees=45, twistLimitDegrees=45}
 {bone1Name=thigh_r, bone2Name=pelvis, swing1LimitDegrees=45, swing2LimitDegrees=45, twistLimitDegrees=45}
 {bone1Name=dyn_suitBack_01, bone2Name=base_suitBack, swing1LimitDegrees=20, swing2LimitDegrees=20, twistLimitDegrees=45, twistMotion=Locked}
 {bone1Name=dyn_suitBack_02, bone2Name=dyn_suitBack_01, swing1LimitDegrees=20, swing2LimitDegrees=20, twistLimitDegrees=45, twistMotion=Locked}
 {
  bone1Name=dyn_suitSideBack_01_l
  bone2Name=base_suitSideBack_l
  swing1LimitDegrees=20
  swing2LimitDegrees=20
  twistLimitDegrees=45
  twistMotion=Locked
 }
 {
  bone1Name=dyn_suitSideBack_02_l
  bone2Name=dyn_suitSideBack_01_l
  swing1LimitDegrees=20
  swing2LimitDegrees=20
  twistLimitDegrees=45
  twistMotion=Locked
 }
 {
  bone1Name=dyn_suitSideBack_01_r
  bone2Name=base_suitSideBack_r
  swing1LimitDegrees=20
  swing2LimitDegrees=20
  twistLimitDegrees=45
  twistMotion=Locked
 }
 {
  bone1Name=dyn_suitSideBack_02_r
  bone2Name=dyn_suitSideBack_01_r
  swing1LimitDegrees=20
  swing2LimitDegrees=20
  twistLimitDegrees=45
  twistMotion=Locked
 }
 {
  bone1Name=dyn_suitSideFront_01_l
  bone2Name=base_suitSideFront_l
  swing1LimitDegrees=22
  swing2LimitDegrees=8
  twistLimitDegrees=45
  twistMotion=Locked
 }
 {
  bone1Name=dyn_suitSideFront_02_l
  bone2Name=dyn_suitSideFront_01_l
  swing1LimitDegrees=30
  swing2LimitDegrees=8
  twistLimitDegrees=45
  twistMotion=Locked
 }
 {
  bone1Name=dyn_suitSideFront_01_r
  bone2Name=base_suitSideFront_r
  swing1LimitDegrees=6
  swing2LimitDegrees=8
  twistLimitDegrees=45
  twistMotion=Locked
 }
 {
  bone1Name=dyn_suitSideFront_02_r
  bone2Name=dyn_suitSideFront_01_r
  swing1LimitDegrees=8
  swing2LimitDegrees=8
  twistLimitDegrees=45
  twistMotion=Locked
 }
 {
  bone1Name=dyn_suitSide_01_l
  bone2Name=base_suitSide_l
  swing1LimitDegrees=20
  swing2LimitDegrees=20
  twistLimitDegrees=45
  twistMotion=Locked
 }
 {
  bone1Name=dyn_suitSide_02_l
  bone2Name=dyn_suitSide_01_l
  swing1LimitDegrees=20
  swing2LimitDegrees=20
  twistLimitDegrees=45
  twistMotion=Locked
 }
 {
  bone1Name=dyn_suitSide_01_r
  bone2Name=base_suitSide_r
  swing1LimitDegrees=30
  swing2LimitDegrees=30
  twistLimitDegrees=45
  twistMotion=Locked
 }
 {
  bone1Name=dyn_suitSide_02_r
  bone2Name=dyn_suitSide_01_r
  swing1LimitDegrees=30
  swing2LimitDegrees=30
  twistLimitDegrees=45
  twistMotion=Locked
 }
 {
  bone1Name=base_suitBack
  bone2Name=pelvis
  swing1LimitDegrees=45
  swing1Motion=Locked
  swing2LimitDegrees=45
  swing2Motion=Locked
  twistLimitDegrees=45
  twistMotion=Locked
 }
 {
  bone1Name=base_suitSideBack_l
  bone2Name=pelvis
  swing1LimitDegrees=45
  swing1Motion=Locked
  swing2LimitDegrees=45
  swing2Motion=Locked
  twistLimitDegrees=45
  twistMotion=Locked
 }
 {
  bone1Name=base_suitSideBack_r
  bone2Name=pelvis
  swing1LimitDegrees=45
  swing1Motion=Locked
  swing2LimitDegrees=45
  swing2Motion=Locked
  twistLimitDegrees=45
  twistMotion=Locked
 }
 {
  bone1Name=base_suitSideFront_l
  bone2Name=pelvis
  swing1LimitDegrees=45
  swing1Motion=Locked
  swing2LimitDegrees=45
  swing2Motion=Locked
  twistLimitDegrees=45
  twistMotion=Locked
 }
 {
  bone1Name=base_suitSideFront_r
  bone2Name=pelvis
  swing1LimitDegrees=45
  swing1Motion=Locked
  swing2LimitDegrees=45
  swing2Motion=Locked
  twistLimitDegrees=45
  twistMotion=Locked
 }
 {
  bone1Name=base_suitSide_l
  bone2Name=pelvis
  swing1LimitDegrees=45
  swing1Motion=Locked
  swing2LimitDegrees=45
  swing2Motion=Locked
  twistLimitDegrees=45
  twistMotion=Locked
 }
 {
  bone1Name=base_suitSide_r
  bone2Name=pelvis
  swing1LimitDegrees=45
  swing1Motion=Locked
  swing2LimitDegrees=45
  swing2Motion=Locked
  twistLimitDegrees=45
  twistMotion=Locked
 }
 {
  bone1Name=deform_glute_l
  bone2Name=pelvis
  swing1LimitDegrees=45
  swing1Motion=Locked
  swing2LimitDegrees=45
  swing2Motion=Locked
  twistLimitDegrees=45
  twistMotion=Locked
 }
 {
  bone1Name=deform_glute_r
  bone2Name=pelvis
  swing1LimitDegrees=45
  swing1Motion=Locked
  swing2LimitDegrees=45
  swing2Motion=Locked
  twistLimitDegrees=45
  twistMotion=Locked
 }
]
