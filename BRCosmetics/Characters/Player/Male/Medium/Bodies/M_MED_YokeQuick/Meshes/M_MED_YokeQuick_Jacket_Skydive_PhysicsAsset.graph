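# Physics asset for the skydive variant of the M_MED_YokeQuick male body's jacket, rigging its lower suit panels for secondary motion.
# Capsules on pelvis, glutes and doubled thighs form body volumes, with glutes and tiny panel-base anchors rotation-locked to the pelvis.
# From those anchors hang seven two-bone capsule chains around back, sides and front.
# Chains have locked twist and narrow swing limits, looser on the front-side panels.
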PhysicsAsset /BRCosmetics/Characters/Player/Male/Medium/Bodies/M_MED_YokeQuick/Meshes/M_MED_YokeQuick_Jacket_Skydive_PhysicsAsset
bodies {
 base_suitBack [{shapeType=Capsule, length=0.505, radius=0.505, shapeName=base_suitBack_capsule, rotation=<rot 0 -90 0>}]
 base_suitSideBack_l [{shapeType=Capsule, length=0.505, radius=0.505, shapeName=base_suitSideBack_l_capsule, rotation=<rot 0 -90 0>}]
 base_suitSideBack_r [{shapeType=Capsule, length=0.505, radius=0.505, shapeName=base_suitSideBack_r_capsule, rotation=<rot 0 -90 0>}]
 base_suitSideFront_l [{shapeType=Capsule, length=0.505, radius=0.505, shapeName=base_suitSideFront_l_capsule, rotation=<rot 0 -90 0>}]
 base_suitSideFront_r [{shapeType=Capsule, length=0.505, radius=0.505, shapeName=base_suitSideFront_r_capsule, rotation=<rot 0 -90 0>}]
 base_suitSide_l [{shapeType=Capsule, length=0.505, radius=0.505, shapeName=base_suitSide_l_capsule, rotation=<rot 0 -90 0>}]
 base_suitSide_r [{shapeType=Capsule, length=0.505, radius=0.505, shapeName=base_suitSide_r_capsule, rotation=<rot 0 -90 0>}]
 deform_glute_l [{
   shapeType=Capsule
   length=3.4775
   radius=4.273
   shapeName=deform_glute_l_capsule
   center=<1.7848 -3.7584 1.723>
   rotation=<rot 0 -90.0001 -30.7074>
  }]
 deform_glute_r [{
   shapeType=Capsule
   length=3.4775
   radius=4.273
   shapeName=deform_glute_r_capsule
   center=<-1.7848 3.7584 -1.723>
   rotation=<rot 0 -90.0001 149.2926>
  }]
 dyn_suitBack_01 [{
   shapeType=Capsule
   length=13.3981
   radius=7.2622
   shapeName=dyn_suitSideFront_02_l_capsule
   center=<9.8084 -7.1596 0.1099>
   rotation=<rot -7.3435 -90.0005 -0.0003>
  }]
 dyn_suitBack_02 [{
   shapeType=Capsule
   length=13.3981
   radius=7.2622
   shapeName=dyn_suitSideFront_02_l_capsule
   center=<14.2569 -7.0141 0>
   rotation=<rot 0 -90 0>
  }]
 dyn_suitSideBack_01_l [{
   shapeType=Capsule
   length=13.3981
   radius=7.2622
   shapeName=dyn_suitSideFront_02_l_capsule
   center=<9.8084 -6.5131 0.1932>
   rotation=<rot -7.3435 -90.0005 -0.0003>
  }]
 dyn_suitSideBack_01_r [{
   shapeType=Capsule
   length=13.3981
   radius=7.2622
   shapeName=dyn_suitSideFront_02_l_capsule
   center=<10.6609 -7.3038 0.0913>
   rotation=<rot -7.3435 -90.0005 -0.0003>
  }]
 dyn_suitSideBack_02_l [{
   shapeType=Capsule
   length=13.3981
   radius=7.2622
   shapeName=dyn_suitSideFront_02_l_capsule
   center=<14.2567 -5.8379 0.0164>
   rotation=<rot -1.138 -89.8395 -0.9259>
  }]
 dyn_suitSideBack_02_r [{
   shapeType=Capsule
   length=13.3981
   radius=7.2622
   shapeName=dyn_suitSideFront_02_l_capsule
   center=<14.2563 -5.7105 0.0257>
   rotation=<rot -1.138 -89.8395 -0.9259>
  }]
 dyn_suitSideFront_01_l [{
   shapeType=Capsule
   length=13.3981
   radius=7.2622
   shapeName=dyn_suitSideFront_02_l_capsule
   center=<9.8084 -7.1596 0.1099>
   rotation=<rot -7.3435 -90.0005 -0.0003>
  }]
 dyn_suitSideFront_01_r [{
   shapeType=Capsule
   length=13.3981
   radius=7.2622
   shapeName=dyn_suitSideFront_02_l_capsule
   center=<9.8083 -7.9985 0.0584>
   rotation=<rot 4.3395 -90.0007 -0.0002>
  }]
 dyn_suitSideFront_02_l [{
   shapeType=Capsule
   length=13.3981
   radius=7.2622
   shapeName=dyn_suitSideFront_02_l_capsule
   center=<14.2567 -5.1896 -0.0786>
   rotation=<rot 8.7728 -90.0012 -0.9364>
  }]
 dyn_suitSideFront_02_r [{
   shapeType=Capsule
   length=13.3981
   radius=7.2622
   shapeName=dyn_suitSideFront_02_l_capsule
   center=<14.2567 -7.2295 0.2362>
   rotation=<rot 8.7728 -90.0012 -0.9364>
  }]
 dyn_suitSide_01_l [{
   shapeType=Capsule
   length=13.3981
   radius=7.2622
   shapeName=dyn_suitSideFront_02_l_capsule
   center=<9.8084 -6.1656 0.238>
   rotation=<rot -7.3435 -90.0005 -0.0003>
  }]
 dyn_suitSide_01_r [{
   shapeType=Capsule
   length=13.3981
   radius=7.2622
   shapeName=dyn_suitSideFront_02_l_capsule
   center=<8.1383 -6.6282 0>
   rotation=<rot 0 -90 0>
  }]
 dyn_suitSide_02_l [{
   shapeType=Capsule
   length=13.3981
   radius=7.2622
   shapeName=dyn_suitSideFront_02_l_capsule
   center=<14.2539 -5.1928 0>
   rotation=<rot -6.6548 -89.7504 -0.9321>
  }]
 dyn_suitSide_02_r [{
   shapeType=Capsule
   length=13.3981
   radius=7.2622
   shapeName=dyn_suitSideFront_02_l_capsule
   center=<14.2567 -5.8505 0>
   rotation=<rot -6.6548 -89.7504 -0.9321>
  }]
 pelvis [{shapeType=Capsule, length=10.8729, radius=10.3773, center=<3.3821 -0.5468 0.2718>, rotation=<rot 4.4575 -90.0015 177.6591>}]
 thigh_l [
  {shapeType=Capsule, length=29.9507, radius=7.2741, center=<-16.8369 -1.8512 0.0429>, rotation=<rot 0.2482 -92.6907 -89.0955>}
  {shapeType=Capsule, length=29.9507, radius=7.2741, center=<-16.9984 0.986 -0.2397>, rotation=<rot 0.1933 -89.2468 -89.0822>}
 ]
 thigh_r [
  {shapeType=Capsule, length=29.9507, radius=7.2741, center=<16.856 1.8555 1.1526>, rotation=<rot 0.2482 -92.6907 90.9045>}
  {shapeType=Capsule, length=29.9507, radius=7.2741, center=<17.0002 -1.8541 0.3399>, rotation=<rot 0.1532 -86.7572 90.9253>}
 ]
}
constraints [
 {bone1Name=thigh_l, bone2Name=pelvis, swing1LimitDegrees=45, swing2LimitDegrees=45, twistLimitDegrees=45}
 {bone1Name=thigh_r, bone2Name=pelvis, swing1LimitDegrees=45, swing2LimitDegrees=45, twistLimitDegrees=45}
 {bone1Name=dyn_suitBack_01, bone2Name=base_suitBack, swing1LimitDegrees=6, swing2LimitDegrees=5, twistLimitDegrees=45, twistMotion=Locked}
 {bone1Name=dyn_suitBack_02, bone2Name=dyn_suitBack_01, swing1LimitDegrees=6, swing2LimitDegrees=5, twistLimitDegrees=45, twistMotion=Locked}
 {
  bone1Name=dyn_suitSideBack_01_l
  bone2Name=base_suitSideBack_l
  swing1LimitDegrees=6
  swing2LimitDegrees=5
  twistLimitDegrees=45
  twistMotion=Locked
 }
 {
  bone1Name=dyn_suitSideBack_02_l
  bone2Name=dyn_suitSideBack_01_l
  swing1LimitDegrees=6
  swing2LimitDegrees=5
  twistLimitDegrees=45
  twistMotion=Locked
 }
 {
  bone1Name=dyn_suitSideBack_01_r
  bone2Name=base_suitSideBack_r
  swing1LimitDegrees=6
  swing2LimitDegrees=5
  twistLimitDegrees=45
  twistMotion=Locked
 }
 {
  bone1Name=dyn_suitSideBack_02_r
  bone2Name=dyn_suitSideBack_01_r
  swing1LimitDegrees=6
  swing2LimitDegrees=5
  twistLimitDegrees=45
  twistMotion=Locked
 }
 {
  bone1Name=dyn_suitSideFront_01_l
  bone2Name=base_suitSideFront_l
  swing1LimitDegrees=15
  swing2LimitDegrees=5
  twistLimitDegrees=45
  twistMotion=Locked
 }
 {
  bone1Name=dyn_suitSideFront_02_l
  bone2Name=dyn_suitSideFront_01_l
  swing1LimitDegrees=15
  swing2LimitDegrees=5
  twistLimitDegrees=45
  twistMotion=Locked
 }
 {
  bone1Name=dyn_suitSideFront_01_r
  bone2Name=base_suitSideFront_r
  swing1LimitDegrees=15
  swing2LimitDegrees=5
  twistLimitDegrees=45
  twistMotion=Locked
 }
 {
  bone1Name=dyn_suitSideFront_02_r
  bone2Name=dyn_suitSideFront_01_r
  swing1LimitDegrees=15
  swing2LimitDegrees=5
  twistLimitDegrees=45
  twistMotion=Locked
 }
 {bone1Name=dyn_suitSide_01_l, bone2Name=base_suitSide_l, swing1LimitDegrees=6, swing2LimitDegrees=5, twistLimitDegrees=45, twistMotion=Locked}
 {
  bone1Name=dyn_suitSide_02_l
  bone2Name=dyn_suitSide_01_l
  swing1LimitDegrees=6
  swing2LimitDegrees=5
  twistLimitDegrees=45
  twistMotion=Locked
 }
 {bone1Name=dyn_suitSide_01_r, bone2Name=base_suitSide_r, swing1LimitDegrees=6, swing2LimitDegrees=5, twistLimitDegrees=45, twistMotion=Locked}
 {
  bone1Name=dyn_suitSide_02_r
  bone2Name=dyn_suitSide_01_r
  swing1LimitDegrees=6
  swing2LimitDegrees=5
  twistLimitDegrees=45
  twistMotion=Locked
 }
 {
  bone1Name=base_suitBack
  bone2Name=pelvis
  swing1LimitDegrees=45
  swing1Motion=Locked
  swing2LimitDegrees=45
  swing2Motion=Locked
  twistLimitDegrees=45
  twistMotion=Locked
 }
 {
  bone1Name=base_suitSideBack_l
  bone2Name=pelvis
  swing1LimitDegrees=45
  swing1Motion=Locked
  swing2LimitDegrees=45
  swing2Motion=Locked
  twistLimitDegrees=45
  twistMotion=Locked
 }
 {
  bone1Name=base_suitSideBack_r
  bone2Name=pelvis
  swing1LimitDegrees=45
  swing1Motion=Locked
  swing2LimitDegrees=45
  swing2Motion=Locked
  twistLimitDegrees=45
  twistMotion=Locked
 }
 {
  bone1Name=base_suitSideFront_l
  bone2Name=pelvis
  swing1LimitDegrees=45
  swing1Motion=Locked
  swing2LimitDegrees=45
  swing2Motion=Locked
  twistLimitDegrees=45
  twistMotion=Locked
 }
 {
  bone1Name=base_suitSideFront_r
  bone2Name=pelvis
  swing1LimitDegrees=45
  swing1Motion=Locked
  swing2LimitDegrees=45
  swing2Motion=Locked
  twistLimitDegrees=45
  twistMotion=Locked
 }
 {
  bone1Name=base_suitSide_l
  bone2Name=pelvis
  swing1LimitDegrees=45
  swing1Motion=Locked
  swing2LimitDegrees=45
  swing2Motion=Locked
  twistLimitDegrees=45
  twistMotion=Locked
 }
 {
  bone1Name=base_suitSide_r
  bone2Name=pelvis
  swing1LimitDegrees=45
  swing1Motion=Locked
  swing2LimitDegrees=45
  swing2Motion=Locked
  twistLimitDegrees=45
  twistMotion=Locked
 }
 {
  bone1Name=deform_glute_l
  bone2Name=pelvis
  swing1LimitDegrees=45
  swing1Motion=Locked
  swing2LimitDegrees=45
  swing2Motion=Locked
  twistLimitDegrees=45
  twistMotion=Locked
 }
 {
  bone1Name=deform_glute_r
  bone2Name=pelvis
  swing1LimitDegrees=45
  swing1Motion=Locked
  swing2LimitDegrees=45
  swing2Motion=Locked
  twistLimitDegrees=45
  twistMotion=Locked
 }
]
