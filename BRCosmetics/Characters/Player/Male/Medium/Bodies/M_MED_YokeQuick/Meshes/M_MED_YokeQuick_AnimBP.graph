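# Post-process-style cosmetic AnimBP for the YokeQuick male body, animating suit jacket, tie and buttons over the attached parent's pose.
# A state-driven Control Rig feeds a cached pose into skydive and regular jacket RigidBody sims, selected by bIsSkyDiving, then a tie sim.
# Additive bone offsets, weighted constraints and look-ats adjust cloth bones and buttons.
# A front-end idle curve blends slot animation onto tie and jacket bones.
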AnimBlueprint /BRCosmetics/Characters/Player/Male/Medium/Bodies/M_MED_YokeQuick/Meshes/M_MED_YokeQuick_AnimBP
target_skeleton=&/M_MED_YokeQuick_Skeleton
graphs [
 AnimGraph {nodes [
   "Transform (Modify) Bone - Bone: base_tie"!(ComponentPose, bAlphaBoolEnabled:bool=True, Translation="0,0,0", Rotation="0.237,0.000,-4.654", Scale="1,1,1") -> (Pose) {
    alphaInputType=Bool
    rotationMode=BMM_Additive
    translationMode=BMM_Additive
    alphaBoolBlend {blendInTime=1, blendOutTime=1}
    boneToModify {boneName=base_tie}
    rotation {pitch=0.237, roll=-4.654}
   }
   n1 = RigidBody!(ComponentPose=Local To Component(LocalPose=Use cached pose 'PreSim'()), Alpha, OverrideWorldGravity:Vector="0,0,0", ExternalForce) -> (Pose) {
    bOverrideWorldGravity=true
    overridePhysicsAsset=&/M_MED_YokeQuick_Jacket_Skydive_PhysicsAsset
    simulationSpace=BaseBoneSpace
    baseBoneRef {boneName=pelvis}
    componentAppliedLinearAccClamp=<1000 1000 1000>
    simSpaceSettings {
     dampingAlpha=0.6
     maxAngularAcceleration=20
     maxAngularVelocity=5
     maxLinearAcceleration=1000
     maxLinearVelocity=300
     velocityScaleZ=0.5
     worldAlpha=0.5
     externalLinearDragV=<0.3 0.3 0.3>
    }
   }
   n2 = Component To Local(ComponentPose=@n1)
   n3 = RigidBody(ComponentPose=Local To Component(LocalPose=Use cached pose 'PreSim'()), OverrideWorldGravity="0,0,0") {
    bOverrideWorldGravity=true
    overridePhysicsAsset=&/M_MED_YokeQuick_Jacket_PhysicsAsset
    simulationSpace=BaseBoneSpace
    baseBoneRef {boneName=pelvis}
    componentAppliedLinearAccClamp=<3000 3000 3000>
    simSpaceSettings {
     dampingAlpha=0.1
     maxAngularAcceleration=30
     maxAngularVelocity=6
     maxLinearAcceleration=1000
     maxLinearVelocity=300
     velocityScaleZ=0.5
     worldAlpha=0.6
     externalLinearDragV=<0.3 0.3 1>
    }
   }
   n4 = Component To Local(ComponentPose=@n3)
   n5 = Get bIsSkyDiving()
   n6 = RigidBody(
    ComponentPose=Local To Component(LocalPose=Blend Poses by bool(BlendPose_0=@n2, BlendPose_1=@n4, bActiveValue=@n5) {blendType=Linear})
    OverrideWorldGravity="0,0,0"
   ) {
    bOverrideWorldGravity=true
    overridePhysicsAsset=&/M_MED_YokeQuick_Tiee_Physics
    simulationSpace=BaseBoneSpace
    baseBoneRef {boneName=spine_05}
    componentAppliedLinearAccClamp=<1000 1000 1000>
    simSpaceSettings {
     dampingAlpha=0.5
     maxAngularAcceleration=20
     maxAngularVelocity=5
     maxLinearAcceleration=1000
     maxLinearVelocity=300
     velocityScaleZ=0.25
     worldAlpha=1
     externalLinearDragV=<0.3 0.3 0.3>
    }
   }
   n7 = "Transform (Modify) Bone - Bone: base_suitSideFront_l"!(ComponentPose=@n6, bAlphaBoolEnabled:bool=True, Translation="0,0,0", Rotation="-0.000,-14.285,0.000", Scale="1,1,1") -> (Pose) {
    alphaInputType=Bool
    rotationMode=BMM_Additive
    translationMode=BMM_Additive
    alphaBoolBlend {blendInTime=0.2, blendOutTime=0.2}
    boneToModify {boneName=base_suitSideFront_l}
    rotation {yaw=-14.285}
   }
   n8 = "Transform (Modify) Bone - Bone: dyn_suitSideFront_01_r"!(ComponentPose=@n7, bAlphaBoolEnabled:bool=True, Translation="-0.659,0.000,-0.053", Rotation="3.508,0.000,0.000", Scale="1,1,1") -> (Pose) {
    alphaInputType=Bool
    rotationMode=BMM_Additive
    translationMode=BMM_Additive
    alphaBoolBlend {blendInTime=0.5, blendOutTime=0.5}
    boneToModify {boneName=dyn_suitSideFront_01_r}
    rotation {pitch=3.508}
    translation {x=-0.659, z=-0.053}
   }
   n9 = "Transform (Modify) Bone - Bone: dyn_suitSide_02_l"!(ComponentPose=@n8, bAlphaBoolEnabled:bool=True, Translation="0,0,0", Rotation="1.092,-15.739,-2.339", Scale="1,1,1") -> (Pose) {
    alphaInputType=Bool
    rotationMode=BMM_Additive
    translationMode=BMM_Additive
    alphaBoolBlend {blendInTime=0.5, blendOutTime=0.5}
    boneToModify {boneName=dyn_suitSide_02_l}
    rotation=<rot 1.092 -15.739 -2.339>
   }
   n10 = "Transform (Modify) Bone - Bone: dyn_suitSideFront_02_l"!(ComponentPose=@n9, bAlphaBoolEnabled:bool=True, Translation="0,0,0", Rotation="-20.951,0.000,-0.000", Scale="1,1,1") -> (Pose) {
    alphaInputType=Bool
    rotationMode=BMM_Additive
    translationMode=BMM_Additive
    alphaBoolBlend {blendInTime=0.2, blendOutTime=0.2}
    boneToModify {boneName=dyn_suitSideFront_02_l}
    rotation {pitch=-20.951}
   }
   n11 = "Transform (Modify) Bone - Bone: dyn_suitSide_02_l"(ComponentPose=@n10, bAlphaBoolEnabled=True, Translation="0,0,0", Rotation="0,0,-5", Scale="1,1,1") {
    alphaInputType=Bool
    rotationMode=BMM_Additive
    translationMode=BMM_Additive
    alphaBoolBlend {blendInTime=0.2, blendOutTime=0.2}
    boneToModify {boneName=dyn_suitSide_02_l}
    rotation {roll=-5}
   }
   n12 = "Transform (Modify) Bone - Bone: dyn_suitSideFront_01_l"!(ComponentPose=@n11, bAlphaBoolEnabled:bool=True, Translation="0,0,-1", Rotation="-1.280,-5.001,-0.000", Scale="1,1,1") -> (Pose) {
    alphaInputType=Bool
    rotationMode=BMM_Additive
    translationMode=BMM_Additive
    alphaBoolBlend {blendInTime=0.5, blendOutTime=0.5}
    boneToModify {boneName=dyn_suitSideFront_01_l}
    rotation {pitch=-1.28, yaw=-5.001}
    translation {z=-1}
   }
   n13 = "Transform (Modify) Bone - Bone: dyn_suitSideFront_02_l"(ComponentPose=@n12, bAlphaBoolEnabled=True, Translation="0,0,0", Rotation="-6.888,12.364,-1.507", Scale="1,1,1") {
    alphaInputType=Bool
    rotationMode=BMM_Additive
    translationMode=BMM_Additive
    alphaBoolBlend {blendInTime=0.5, blendOutTime=0.5}
    boneToModify {boneName=dyn_suitSideFront_02_l}
    rotation=<rot -6.888 12.364 -1.507>
   }
   n14 = "Transform (Modify) Bone - Bone: dyn_suitSide_02_l"(ComponentPose=@n13, bAlphaBoolEnabled=True, Translation="0,0,0", Rotation="-2.462,-13.525,12.115", Scale="1,1,1") {
    alphaInputType=Bool
    rotationMode=BMM_Additive
    translationMode=BMM_Additive
    alphaBoolBlend {blendInTime=0.5, blendOutTime=0.5}
    boneToModify {boneName=dyn_suitSide_02_l}
    rotation=<rot -2.462 -13.525 12.115>
   }
   n15 = "Transform (Modify) Bone - Bone: dyn_suitSideFront_02_r"!(ComponentPose=@n14, bAlphaBoolEnabled:bool=True, Translation="-0.000,0.000,2.520", Rotation="9.851,0.000,0.000", Scale="1,1,1") -> (Pose) {
    alphaInputType=Bool
    rotationMode=BMM_Additive
    translationMode=BMM_Additive
    alphaBoolBlend {blendInTime=0.5, blendOutTime=0.5}
    boneToModify {boneName=dyn_suitSideFront_02_r}
    rotation {pitch=9.851}
    translation {z=2.52}
   }
   n16 = "Transform (Modify) Bone - Bone: dyn_tie_02"!(ComponentPose=@n15, bAlphaBoolEnabled:bool=True, Translation="0,0,0", Rotation="0.229,-0.027,-13.027", Scale="1,1,1") -> (Pose) {
    alphaInputType=Bool
    rotationMode=BMM_Additive
    translationMode=BMM_Additive
    alphaBoolBlend {blendInTime=1, blendOutTime=1}
    boneToModify {boneName=dyn_tie_02}
    rotation=<rot 0.229 -0.027 -13.027>
   }
   n17 = "Transform (Modify) Bone - Bone: dyn_tie_03"!(ComponentPose=@n16, bAlphaBoolEnabled:bool=True, Translation="0,0,0", Rotation="0.229,-0.019,-12.235", Scale="1,1,1") -> (Pose) {
    alphaInputType=Bool
    rotationMode=BMM_Additive
    translationMode=BMM_Additive
    alphaBoolBlend {blendInTime=1, blendOutTime=1}
    boneToModify {boneName=dyn_tie_03}
    rotation=<rot 0.229 -0.019 -12.235>
   }
   n18 = "Transform (Modify) Bone - Bone: dyn_suitSideFront_01_r"(ComponentPose=@n17, bAlphaBoolEnabled=True, Translation="0,0,0", Rotation="0.017,-4.922,0.032", Scale="1,1,1") {
    alphaInputType=Bool
    rotationMode=BMM_Additive
    rotationSpace=BCS_BoneSpace
    scaleSpace=BCS_BoneSpace
    translationMode=BMM_Additive
    translationSpace=BCS_BoneSpace
    alphaBoolBlend {blendInTime=0.2, blendOutTime=0.2}
    boneToModify {boneName=dyn_suitSideFront_01_r}
    rotation=<rot 0.017 -4.922 0.032>
   }
   n19 = "Transform (Modify) Bone - Bone: base_suitSide_r"!(ComponentPose=@n18, bAlphaBoolEnabled:bool=True, Translation="0,0,0", Rotation="-9.299,-4.376,-0.007", Scale="1,1,1") -> (Pose) {
    alphaInputType=Bool
    rotationMode=BMM_Additive
    rotationSpace=BCS_BoneSpace
    scaleSpace=BCS_BoneSpace
    translationMode=BMM_Additive
    translationSpace=BCS_BoneSpace
    alphaBoolBlend {blendInTime=0.6, blendOutTime=1}
    boneToModify {boneName=base_suitSide_r}
    rotation=<rot -9.299 -4.376 -0.007>
   }
   n20 = "Transform (Modify) Bone - Bone: dyn_suitSideFront_02_r"(ComponentPose=@n19, bAlphaBoolEnabled=True, Translation="0,0,0", Rotation="4.978,16,0.7", Scale="1,1,1") {
    alphaInputType=Bool
    rotationMode=BMM_Additive
    rotationSpace=BCS_BoneSpace
    scaleSpace=BCS_BoneSpace
    translationMode=BMM_Additive
    translationSpace=BCS_BoneSpace
    alphaBoolBlend {blendInTime=0.6, blendOutTime=1}
    boneToModify {boneName=dyn_suitSideFront_02_r}
    rotation=<rot 4.978 16 0.7>
   }
   n21 = "Transform (Modify) Bone - Bone: base_suitSideFront_r"!(ComponentPose=@n20, bAlphaBoolEnabled:bool=True, Translation="0,0,0", Rotation="0.974,-15,-0.059", Scale="1,1,1") -> (Pose) {
    alphaInputType=Bool
    rotationMode=BMM_Additive
    rotationSpace=BCS_BoneSpace
    scaleSpace=BCS_BoneSpace
    translationMode=BMM_Additive
    translationSpace=BCS_BoneSpace
    alphaBoolBlend {blendInTime=0.2, blendOutTime=1}
    boneToModify {boneName=base_suitSideFront_r}
    rotation=<rot 0.974 -15 -0.059>
   }
   n22 = "Transform (Modify) Bone - Bone: dyn_suitSideFront_01_l"(ComponentPose=@n21, bAlphaBoolEnabled=True, Translation="0,0,0", Rotation="0.010,-4.817,-2.829", Scale="1,1,1") {
    alphaInputType=Bool
    rotationMode=BMM_Additive
    rotationSpace=BCS_BoneSpace
    scaleSpace=BCS_BoneSpace
    translationMode=BMM_Additive
    translationSpace=BCS_BoneSpace
    alphaBoolBlend {blendInTime=1, blendOutTime=1}
    boneToModify {boneName=dyn_suitSideFront_01_l}
    rotation=<rot 0.01 -4.817 -2.829>
   }
   n23 = "Transform (Modify) Bone - Bone: dyn_suitSideFront_01_r"(ComponentPose=@n22, bAlphaBoolEnabled=True, Translation="0,0,0", Rotation="0.010,-4.817,-2.829", Scale="1,1,1") {
    alphaInputType=Bool
    rotationMode=BMM_Additive
    rotationSpace=BCS_BoneSpace
    scaleSpace=BCS_BoneSpace
    translationMode=BMM_Additive
    translationSpace=BCS_BoneSpace
    alphaBoolBlend {blendInTime=1, blendOutTime=1}
    boneToModify {boneName=dyn_suitSideFront_01_r}
    rotation=<rot 0.01 -4.817 -2.829>
   }
   n24 = "Constraint - base_suit_button_01 (PPPPPPPPP)"!(
    ComponentPose=@n23
    Alpha=1
    ConstraintWeights_0=0.3422
    ConstraintWeights_1:float=0.2514
    ConstraintWeights_2:float=0.1898
    ConstraintWeights_3:float=0.1653
    ConstraintWeights_4:float=0.0268
    ConstraintWeights_5:float=0.022
    ConstraintWeights_6:float=0.0009
    ConstraintWeights_7:float=0.0009
    ConstraintWeights_8:float=0.0008
   ) -> (Pose) {
    boneToModify {boneName=base_suit_button_01}
    constraintSetup [
     {targetBone {boneName=pelvis}}
     {targetBone {boneName=spine_02}}
     {targetBone {boneName=spine_01}}
     {targetBone {boneName=spine_03}}
     {targetBone {boneName=dyn_suitSideFront_01_r}}
     {targetBone {boneName=spine_04}}
     {targetBone {boneName=deform_pec_aim_l}}
     {targetBone {boneName=deform_pec_aim_r}}
     {targetBone {boneName=dyn_suitSide_01_r}}
    ]
    constraintWeights [1,1,1,1,1,1,1,1,1]
   }
   n25 = "Constraint - base_suit_button_01_LookAt (PPPPPPPPP)"!(
    ComponentPose=@n24
    Alpha=1
    ConstraintWeights_0=0.3853
    ConstraintWeights_1:float=0.2122
    ConstraintWeights_2:float=0.1728
    ConstraintWeights_3:float=0.1389
    ConstraintWeights_4:float=0.0705
    ConstraintWeights_5:float=0.0161
    ConstraintWeights_6:float=0.003
    ConstraintWeights_7:float=0.0007
    ConstraintWeights_8:float=0.0007
   ) -> (Pose) {
    boneToModify {boneName=base_suit_button_01_LookAt}
    constraintSetup [
     {targetBone {boneName=pelvis}}
     {targetBone {boneName=spine_02}}
     {targetBone {boneName=spine_01}}
     {targetBone {boneName=spine_03}}
     {targetBone {boneName=dyn_suitSideFront_01_r}}
     {targetBone {boneName=spine_04}}
     {targetBone {boneName=dyn_suitSide_01_r}}
     {targetBone {boneName=deform_pec_aim_l}}
     {targetBone {boneName=deform_pec_aim_r}}
    ]
    constraintWeights [1,1,1,1,1,1,1,1,1]
   }
   n26 = "Constraint - base_suit_button_02 (PPPPPPPP)"!(
    ComponentPose=@n25
    Alpha=1
    ConstraintWeights_0=0.392
    ConstraintWeights_1:float=0.2931
    ConstraintWeights_2:float=0.1741
    ConstraintWeights_3:float=0.048
    ConstraintWeights_4:float=0.0363
    ConstraintWeights_5:float=0.0258
    ConstraintWeights_6:float=0.0163
    ConstraintWeights_7:float=0.0143
   ) -> (Pose) {
    boneToModify {boneName=base_suit_button_02}
    constraintSetup [
     {targetBone {boneName=pelvis}}
     {targetBone {boneName=dyn_suitSideFront_01_r}}
     {targetBone {boneName=dyn_suitSideFront_02_r}}
     {targetBone {boneName=deform_groin_r}}
     {targetBone {boneName=spine_01}}
     {targetBone {boneName=spine_02}}
     {targetBone {boneName=thigh_twist_01_r}}
     {targetBone {boneName=spine_03}}
    ]
    constraintWeights [1,1,1,1,1,1,1,1]
   }
   n27 = "Constraint - base_suit_button_02_LookAt (PPPPPPPP)"!(
    ComponentPose=@n26
    Alpha=1
    ConstraintWeights_0=0.3515
    ConstraintWeights_1:float=0.273
    ConstraintWeights_2:float=0.2568
    ConstraintWeights_3:float=0.0698
    ConstraintWeights_4:float=0.0227
    ConstraintWeights_5:float=0.0122
    ConstraintWeights_6:float=0.0074
    ConstraintWeights_7:float=0.0067
   ) -> (Pose) {
    boneToModify {boneName=base_suit_button_02_LookAt}
    constraintSetup [
     {targetBone {boneName=dyn_suitSideFront_02_r}}
     {targetBone {boneName=pelvis}}
     {targetBone {boneName=dyn_suitSideFront_01_r}}
     {targetBone {boneName=deform_groin_r}}
     {targetBone {boneName=thigh_twist_01_r}}
     {targetBone {boneName=spine_01}}
     {targetBone {boneName=spine_02}}
     {targetBone {boneName=dyn_suitSide_02_r}}
    ]
    constraintWeights [1,1,1,1,1,1,1,1]
   }
   n28 = Look At - Bone: base_suit_button_01(ComponentPose=@n27, Alpha=1) {
    alphaBoolBlend {blendOption=Linear}
    boneToModify {boneName=base_suit_button_01}
    lookAtLocation {x=0}
    lookAtTarget {boneReference {boneName=base_suit_button_01_LookAt}}
   }
   n29 = Look At - Bone: base_suit_button_02(ComponentPose=@n28, Alpha=1) {
    alphaBoolBlend {blendOption=Linear}
    boneToModify {boneName=base_suit_button_02}
    lookAtLocation {x=0}
    lookAtTarget {boneReference {boneName=base_suit_button_02_LookAt}}
   }
   n30 = Layered blend per bone(
    BasePose=Component To Local(ComponentPose=@n29)
    BlendPoses_0=Slot 'DefaultSlot'()
    BlendWeights_0=Get Custom Idle Is Playing Float() -> (Custom Idle Is Playing Float:double)
   ) {layerSetup [{branchFilters [
       {boneName=base_tie}
       {boneName=dyn_suitSideFront_01_l}
       {boneName=dyn_suitSide_01_l}
       {boneName=dyn_suitSideFront_01_r}
       {boneName=dyn_suitSide_01_r}
      ]}]}
   n31 = "Transform (Modify) Bone - Bone: dyn_tie_05"(ComponentPose=Local To Component(LocalPose=@n30), Alpha=1, Translation="0,0,0", Rotation="3,0,0", Scale="1,1,1") {
    rotationMode=BMM_Additive
    rotationSpace=BCS_BoneSpace
    translationMode=BMM_Additive
    translationSpace=BCS_BoneSpace
    boneToModify {boneName=dyn_tie_05}
    rotation {pitch=3}
   }
   Output Pose(Result=Component To Local(ComponentPose=@n31)) {layerGroup=None}
   n32 = Control Rig!(
    Source=Copy Pose From Mesh() {bUseAttachedParent=true}
    Alpha
    isCrouching_CR:bool=Get bIsCrouching()
    isKneeSliding_CR:bool=Get bIsKneeSliding()
    isBaseballSliding_CR:bool=Get bIsBaseballSliding()
    isFE_CR:bool=Get bIsInFrontEnd()
    isDBNO_CR:bool=Get bIsDBNO()
   ) -> (Pose) {controlRigAssetReference {blueprintRigClass=&/M_MED_YokeQuick_ControlRig.M_MED_YokeQuick_ControlRig_C}}
   Save cached pose 'PreSim'(Pose=@n32) {cacheName=PreSim}
   Get isCrouchingAndMoving() -> (isCrouchingAndMoving:bool)
  ]}
 EventGraph {nodes [
   Event BlueprintUpdateAnimation()
   > n1 = "Gravity Override State Function (Parameter Asset) (Nativized)"(
    CharacterPartAnimInstance=Self-Reference()
    GameTimeInSeconds=GetGameTimeInSeconds()
    DebugJointName=spine_05
    Parameters=&/M_MED_YokeQuick_Parameters
   )
   > Set Gravity Override(Gravity Override=@n1)
   n2 = Get bIsMoving2D()
   n3 = Get bIsCrouching()
   > Set isCrouchingNotMoving(isCrouchingNotMoving:bool=AND Boolean(A=@n3, B=NOT Boolean(A=@n2))) -> (Output_Get:bool)
   > Set isCrouchingAndMoving(isCrouchingAndMoving:bool=AND Boolean(A=@n3, B=@n2)) -> (Output_Get:bool)
   > Branch(Condition=Get bIsInFrontEnd())
   > Set Custom Idle Is Playing Float(Custom Idle Is Playing Float:double=GetCurveValue(self=GetOwnerFrontEndAnimInstance(), CurveName=Frontend_Idle_Bypass)) -> (Output_Get:double)
  ]}
]
variables [
 Custom Idle Is Playing Float {type=number}
 Gravity Override {type=Vector}
 isCrouchingAndMoving {type=boolean}
 isCrouchingNotMoving {type=boolean}
]
