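# Physics asset for the M_MED_YokeQuick male medium player body that, by its name, supplies collision shapes for a cape.
# Capsules cover pelvis, spine, neck and clavicles, with extra capsules broadening spine_02 and spine_05, and a sphere covers the head.
# Two-segment dyn_suit bones on the back, sides and side-fronts each carry a capsule.
# Arm, leg and foot bodies have no shapes, and the asset defines no constraints.
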PhysicsAsset /BRCosmetics/Characters/Player/Male/Medium/Bodies/M_MED_YokeQuick/Meshes/M_MED_YokeQuick_CapeColliders
bodies {
 calf_l []
 calf_r []
 clavicle_l [{shapeType=Capsule, length=9.0192, radius=9.8871, center=<9.406 -0.7711 -2.4991>, rotation=<rot 69.3297 1.4476 176.9462>}]
 clavicle_r [{shapeType=Capsule, length=9.0192, radius=9.8871, center=<-9.406 0.7711 2.4991>, rotation=<rot 69.3297 1.4476 -3.0538>}]
 dyn_suitBack_01 [{shapeType=Capsule, length=10, radius=6.0316, shapeName=dyn_suitBack_01_capsule, center=<10.3028 3.6699 0>}]
 dyn_suitBack_02 [{shapeType=Capsule, length=10, radius=5.5978, shapeName=dyn_suitBack_02_capsule, center=<11.6743 2.8273 0>}]
 dyn_suitSideBack_01_l [{
   shapeType=Capsule
   length=10
   radius=6.0316
   shapeName=dyn_suitSideBack_01_l_capsule
   center=<10.3028 3.7968 0.0647>
   rotation=<rot 0.0001 0.0001 -4.284>
  }]
 dyn_suitSideBack_01_r [{
   shapeType=Capsule
   length=10
   radius=6.0316
   shapeName=dyn_suitSideBack_01_r_capsule
   center=<10.3028 3.797 -0.0623>
   rotation=<rot -0.0001 0.0001 4.1231>
  }]
 dyn_suitSideBack_02_l [{
   shapeType=Capsule
   length=10
   radius=5.5978
   shapeName=dyn_suitSideBack_02_l_capsule
   center=<11.3641 3.2792 0.0391>
   rotation=<rot 0.0001 0.0002 -6.4466>
  }]
 dyn_suitSideBack_02_r [{
   shapeType=Capsule
   length=10
   radius=5.5978
   shapeName=dyn_suitSideBack_02_r_capsule
   center=<11.3641 3.2792 -0.0391>
   rotation=<rot -0.0001 0.0002 6.4466>
  }]
 dyn_suitSideFront_01_l [{shapeType=Capsule, length=10, radius=5, shapeName=dyn_suitSideFront_01_l_capsule, center=<10.3028 2.933 0>}]
 dyn_suitSideFront_01_r [{shapeType=Capsule, length=10, radius=5, shapeName=dyn_suitSideFront_01_r_capsule, center=<10.3028 2.933 0>}]
 dyn_suitSideFront_02_l [{shapeType=Capsule, length=10, radius=5, shapeName=dyn_suitSideFront_02_l_capsule, center=<10.3028 2.933 0>}]
 dyn_suitSideFront_02_r [{shapeType=Capsule, length=10, radius=5, shapeName=dyn_suitSideFront_02_r_capsule, center=<10.3028 2.933 0>}]
 dyn_suitSide_01_l [{shapeType=Capsule, length=10, radius=5, shapeName=dyn_suitSide_01_l_capsule, center=<10.3028 2.933 0>}]
 dyn_suitSide_01_r [{shapeType=Capsule, length=10, radius=5, shapeName=dyn_suitSide_01_r_capsule, center=<10.3028 2.933 0>}]
 dyn_suitSide_02_l [{shapeType=Capsule, length=10, radius=5, shapeName=dyn_suitSide_02_l_capsule, center=<10.3028 2.933 0>}]
 dyn_suitSide_02_r [{shapeType=Capsule, length=10, radius=5, shapeName=dyn_suitSide_02_r_capsule, center=<10.3028 2.933 0>}]
 foot_l []
 foot_r []
 head [{shapeType=Sphere, radius=17.3125, center=<11.028 -0.174 0>}]
 lowerarm_l []
 lowerarm_r []
 neck_02 [{shapeType=Capsule, length=13.4375, radius=7.3175, center=<-0.2642 -0.3408 0>, rotation=<rot 90 -171.1193 -175.3978>}]
 pelvis [{shapeType=Capsule, length=4.1456, radius=14.7774, center=<-3.197 0.7838 0>}]
 spine_01 [{shapeType=Capsule, length=5.4438, radius=13.0586, center=<0.5328 -0.3608 0>}]
 spine_02 [
  {shapeType=Capsule, length=7.3264, radius=13.6322, center=<2.4979 1.2049 0.0001>}
  {shapeType=Capsule, length=29.3956, radius=11.5807, center=<2.5024 0.4992 3.9825>, rotation=<rot 90 0.0049 0.0049>}
  {shapeType=Capsule, length=29.875, radius=11.1013, center=<2.4704 0.4991 -3.9825>, rotation=<rot 90 -0.0396 -0.0396>}
 ]
 spine_03 [{shapeType=Capsule, length=7.7107, radius=12.2031, center=<6.5109 0.1071 0.0001>}]
 spine_04 [{shapeType=Capsule, length=8.4404, radius=11.5553, center=<6.2482 -1.1516 0.0002>}]
 spine_05 [
  {shapeType=Capsule, length=5.5226, radius=13.9389, center=<0.2717 2.1811 0>}
  {shapeType=Capsule, length=13.9518, radius=12.5861, center=<-0.2916 0.4452 0>}
 ]
 thigh_l []
 thigh_r []
 upperarm_l []
 upperarm_r []
}
constraints []
