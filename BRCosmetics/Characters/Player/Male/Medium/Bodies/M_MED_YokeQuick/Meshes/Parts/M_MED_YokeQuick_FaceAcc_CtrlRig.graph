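# Control Rig for the YokeQuick male face accessory that corrects hair base bones, the parents of the dyn_hair chains, as the body moves.
# A three-branch sequence first stores spherical pose reader alphas from head, neck, clavicles and right upper arm; only head-back and neck look-up drive corrections here.
# It then layers weighted transform offsets for those alphas and skydiving, floating and parachute states.
# Finally, head swing relative to spine_05 drives offsets in both directions, faded out by emoting.
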ControlRigBlueprint /BRCosmetics/Characters/Player/Male/Medium/Bodies/M_MED_YokeQuick/Meshes/Parts/M_MED_YokeQuick_FaceAcc_CtrlRig
graphs [
 New Function_ContainedGraph {nodes [
   RigVMFunctionEntryNode()
   > RigVMFunctionReturnNode()
  ]}
 RigVMFunctionLibrary {nodes [controls = RigVMCollapseNode(ExecuteContext=())]}
 RigVMFunction_Sequence_1_ContainedGraph {nodes [
   n1 = RigVMFunctionEntryNode()
   RigVMFunction_Sequence = &RigVMFunction_Sequence(ExecuteContext=@n1.ExecuteContext)
   n2 = &RigVMFunction_Sequence(ExecuteContext=@RigVMFunction_Sequence.B)
   RigVMFunctionReturnNode(A=@RigVMFunction_Sequence.A, B=@n2.A, C=@n2.B)
  ]}
 RigVMModel {nodes [
   RigUnit_BeginExecution = &RigUnit_BeginExecution()
   RigVMFunction_Sequence = RigVMAggregateNode(ExecuteContext=@RigUnit_BeginExecution.ExecutePin)
   n1 = &RigUnit_SphericalPoseReader(
    ActiveRegionScaleFactors=(PositiveWidth=0.2,NegativeWidth=1,PositiveHeight=1,NegativeHeight=1)
    ActiveRegionSize=0.1
    Debug=(bDrawDebug=false,bDraw2D=false,bDrawLocalAxes=False,DebugScale=25,DebugSegments=20,DebugThickness=0.25)
    DriverAxis=(X=1,Y=0,Z=0)
    DriverItem=(Type=Bone,Name="head")
    FalloffRegionScaleFactors=(PositiveWidth=1,NegativeWidth=1,PositiveHeight=1,NegativeHeight=1)
    FalloffSize=0.3
    FlipHeightScaling=False
    FlipWidthScaling=False
    OptionalParentItem=(Type=Bone,Name="None")
    RotationOffset=(X=-90,Y=0,Z=35)
    ExecutePin=@RigVMFunction_Sequence.A
   )
   > RigVMVariableNode(Value=@n1.OutputParam) {variable=head_SPR_rot_back_alpha}
   > n2 = &RigUnit_SphericalPoseReader(
    ActiveRegionScaleFactors=(PositiveWidth=0.2,NegativeWidth=1,PositiveHeight=1,NegativeHeight=1)
    ActiveRegionSize=0.1
    Debug=(bDrawDebug=false,bDraw2D=false,bDrawLocalAxes=False,DebugScale=25,DebugSegments=20,DebugThickness=0.25)
    DriverAxis=(X=1,Y=0,Z=0)
    DriverItem=(Type=Bone,Name="neck_01")
    FalloffRegionScaleFactors=(PositiveWidth=1,NegativeWidth=1,PositiveHeight=1,NegativeHeight=1)
    FalloffSize=0.3
    FlipHeightScaling=False
    FlipWidthScaling=False
    OptionalParentItem=(Type=Bone,Name="None")
    RotationOffset=(X=-90,Y=0,Z=35)
   )
   > RigVMVariableNode(Value=@n2.OutputParam) {variable=neck_01_SPR_look_up_alpha}
   > n3 = &RigUnit_SphericalPoseReader(
    ActiveRegionScaleFactors=(PositiveWidth=1,NegativeWidth=1,PositiveHeight=1,NegativeHeight=1)
    ActiveRegionSize=0.05
    Debug=(bDrawDebug=false,bDraw2D=False,bDrawLocalAxes=False,DebugScale=25,DebugSegments=20,DebugThickness=0.25)
    DriverAxis=(X=1,Y=0,Z=0)
    DriverItem=(Type=Bone,Name="clavicle_l")
    FalloffRegionScaleFactors=(PositiveWidth=1,NegativeWidth=1,PositiveHeight=1,NegativeHeight=1)
    FalloffSize=0.3
    FlipHeightScaling=False
    FlipWidthScaling=False
    OptionalParentItem=(Type=Bone,Name="None")
    RotationOffset=(X=-90,Y=0,Z=35)
   )
   > RigVMVariableNode(Value=@n3.OutputParam) {variable=clavicle_l_SPR_rotate_back}
   > n4 = &RigUnit_SphericalPoseReader(
    ActiveRegionScaleFactors=(PositiveWidth=1,NegativeWidth=1,PositiveHeight=1,NegativeHeight=1)
    ActiveRegionSize=0.05
    Debug=(bDrawDebug=false,bDraw2D=False,bDrawLocalAxes=False,DebugScale=25,DebugSegments=20,DebugThickness=0.25)
    DriverAxis=(X=-1,Y=0,Z=0)
    DriverItem=(Type=Bone,Name="clavicle_r")
    FalloffRegionScaleFactors=(PositiveWidth=1,NegativeWidth=1,PositiveHeight=1,NegativeHeight=1)
    FalloffSize=0.3
    FlipHeightScaling=False
    FlipWidthScaling=False
    OptionalParentItem=(Type=Bone,Name="None")
    RotationOffset=(X=-90,Y=0,Z=215)
   )
   > RigVMVariableNode(Value=@n4.OutputParam) {variable=clavicle_r_SPR_rotate_back}
   > n5 = &RigUnit_SphericalPoseReader(
    ActiveRegionScaleFactors=(PositiveWidth=1,NegativeWidth=0.5,PositiveHeight=0.75,NegativeHeight=0.75)
    ActiveRegionSize=0.1
    Debug=(bDrawDebug=false,bDraw2D=False,bDrawLocalAxes=False,DebugScale=25,DebugSegments=20,DebugThickness=0.25)
    DriverAxis=(X=0,Y=1,Z=0)
    DriverItem=(Type=Bone,Name="head")
    FalloffRegionScaleFactors=(PositiveWidth=1,NegativeWidth=0.5,PositiveHeight=1,NegativeHeight=1)
    FalloffSize=0.2
    FlipHeightScaling=False
    FlipWidthScaling=False
    OptionalParentItem=(Type=Bone,Name="None")
    RotationOffset=(X=90,Y=0,Z=-30)
   )
   > RigVMVariableNode(Value=@n5.OutputParam) {variable=head_SPR_look_up_alpha}
   > n6 = &RigUnit_SphericalPoseReader(
    ActiveRegionScaleFactors=(PositiveWidth=1,NegativeWidth=0.5,PositiveHeight=0.75,NegativeHeight=0.75)
    ActiveRegionSize=0.025
    Debug=(bDrawDebug=false,bDraw2D=False,bDrawLocalAxes=False,DebugScale=25,DebugSegments=20,DebugThickness=0.25)
    DriverAxis=(X=-1,Y=0,Z=0)
    DriverItem=(Type=Bone,Name="upperarm_r")
    FalloffRegionScaleFactors=(PositiveWidth=1,NegativeWidth=0.5,PositiveHeight=1,NegativeHeight=1)
    FalloffSize=0.1
    FlipHeightScaling=False
    FlipWidthScaling=False
    OptionalParentItem=(Type=Bone,Name="None")
    RotationOffset=(X=-115,Y=0,Z=25)
   )
   > RigVMVariableNode(Value=@n6.OutputParam) {variable=upperarm_r_SPR_aim_up_two_hands_alpha}
   > RigVMVariableNode(Value=&RigVMFunction_MathDoubleMul(A=RigVMVariableNode() {variable=head_SPR_look_up_alpha}.Value, B=RigVMVariableNode() {variable=upperarm_r_SPR_aim_up_two_hands_alpha}.Value).Result) {variable=look_up_and_left_two_hands_multiplied_alpha}
   > n7 = &RigUnit_SphericalPoseReader(
    ActiveRegionScaleFactors=(PositiveWidth=1,NegativeWidth=1,PositiveHeight=0.5,NegativeHeight=0.5)
    ActiveRegionSize=0.1
    Debug=(bDrawDebug=false,bDraw2D=False,bDrawLocalAxes=False,DebugScale=25,DebugSegments=20,DebugThickness=0.25)
    DriverAxis=(X=0,Y=1,Z=0)
    DriverItem=(Type=Bone,Name="head")
    FalloffRegionScaleFactors=(PositiveWidth=1,NegativeWidth=1,PositiveHeight=0.5,NegativeHeight=0.5)
    FalloffSize=0.2
    FlipHeightScaling=False
    FlipWidthScaling=False
    OptionalParentItem=(Type=Bone,Name="None")
    RotationOffset=(X=110,Y=0,Z=-30)
   )
   > RigVMVariableNode(Value=@n7.OutputParam) {variable=head_SPR_look_left_top_alpha}
   &RigUnit_ModifyTransforms(
    ItemToModify="((Item=(Type=Bone,Name=\"None\"),Transform=(Rotation=(X=0,Y=0,Z=0,W=1),Translation=(X=0,Y=0,Z=0),Scale3D=(X=1,Y=1,Z=1))),(Item=(Type=Bone,Name=\"base_hair_01_back\"),Transform=(Rotation=(X=0,Y=0,Z=0,W=1),Translation=(X=-6,Y=0,Z=0),Scale3D=(X=1,Y=1,Z=1))),(Item ...(truncated)"
    WeightMinimum=0
    ExecutePin=@RigVMFunction_Sequence.B
    Weight=RigVMVariableNode() {variable=head_SPR_rot_back_alpha}.Value
   )
   > &RigUnit_ModifyTransforms(
    ItemToModify="((Item=(Type=Bone,Name=\"None\"),Transform=(Rotation=(X=0,Y=0,Z=0,W=1),Translation=(X=0,Y=0,Z=0),Scale3D=(X=1,Y=1,Z=1))),(Item=(Type=Bone,Name=\"base_hair_01_back\"),Transform=(Rotation=(X=0,Y=0,Z=-0.3455,W=0.9384),Translation=(X=0,Y=0,Z=0),Scale3D=(X=1,Y=1,Z= ...(truncated)"
    WeightMinimum=0
    Weight=RigVMVariableNode() {variable=head_SPR_rot_back_alpha}.Value
   )
   > &RigUnit_ModifyTransforms(
    ItemToModify="((Item=(Type=Bone,Name=\"None\"),Transform=(Rotation=(X=0,Y=0,Z=0,W=1),Translation=(X=0,Y=0,Z=0),Scale3D=(X=1,Y=1,Z=1))),(Item=(Type=Bone,Name=\"base_hair_01_back\"),Transform=(Rotation=(X=-0.0287,Y=0.0035,Z=-0.1735,W=0.9844),Translation=(X=0,Y=0,Z=0),Scale3D= ...(truncated)"
    WeightMinimum=0
    Weight=&RigVMFunction_MathDoubleMul(B=1, A=RigVMVariableNode() {variable=neck_01_SPR_look_up_alpha}.Value).Result
   )
   n8 = RigVMVariableNode() {variable=isSkydiving}
   > &RigUnit_ModifyTransforms(
    ItemToModify="((Item=(Type=Bone,Name=\"None\"),Transform=(Rotation=(X=0,Y=0,Z=0,W=1),Translation=(X=0,Y=0,Z=0),Scale3D=(X=1,Y=1,Z=1))),(Item=(Type=Bone,Name=\"base_hair_01_lf\"),Transform=(Rotation=(X=0,Y=0,Z=0,W=1),Translation=(X=0,Y=0,Z=0),Scale3D=(X=1,Y=1,Z=1))),(Item=(T ...(truncated)"
    WeightMinimum=0
    Weight=&RigVMFunction_MathDoubleMul(A=1, B=&RigVMFunction_MathBoolToFloat(Value=@n8.Value).Result).Result
   )
   n9 = &RigVMFunction_AlphaInterp(
    InterpSpeedDecreasing=10
    InterpSpeedIncreasing=10
    Scale=1
    bClampResult=False
    bInterpResult=False
    Value=&RigVMFunction_MathBoolToFloat(Value=RigVMVariableNode() {variable=isFloating}.Value).Result
   )
   n10 = &RigVMFunction_AnimEvalRichCurve(Curve=(EditorCurveData=(Keys=((InterpMode=RCIM_Cubic),(Time=1,Value=0.3)))), SourceMinimum=0, TargetMinimum=0, Value=@n9.Result)
   > &RigUnit_ModifyTransforms(
    ItemToModify="((Item=(Type=Bone,Name=\"None\"),Transform=(Rotation=(X=0,Y=0,Z=0,W=1),Translation=(X=0,Y=0,Z=0),Scale3D=(X=1,Y=1,Z=1))),(Item=(Type=Bone,Name=\"base_hair_01_back_lf\"),Transform=(Rotation=(X=0.0909,Y=0.0151,Z=0.1637,W=0.9822),Translation=(X=0,Y=0,Z=0),Scale3D ...(truncated)"
    WeightMinimum=0
    Weight=@n10.Result
   )
   > &RigUnit_ModifyTransforms(
    ItemToModify="((Item=(Type=Bone,Name=\"None\"),Transform=(Rotation=(X=0,Y=0,Z=0,W=1),Translation=(X=0,Y=0,Z=0),Scale3D=(X=1,Y=1,Z=1))),(Item=(Type=Bone,Name=\"base_hair_01_lf\"),Transform=(Rotation=(X=0.0284,Y=0.3243,Z=-0.0824,W=0.9419),Translation=(X=0,Y=0,Z=0),Scale3D=(X= ...(truncated)"
    WeightMinimum=0
    Weight=@n10.Result
   )
   > &RigUnit_ModifyTransforms(
    ItemToModify="((Item=(Type=Bone,Name=\"None\"),Transform=(Rotation=(X=0,Y=0,Z=0,W=1),Translation=(X=0,Y=0,Z=0),Scale3D=(X=1,Y=1,Z=1))),(Item=(Type=Bone,Name=\"base_hair_01_lf\"),Transform=(Rotation=(X=0.017,Y=-0.1354,Z=0.1233,W=-0.9829),Translation=(X=0,Y=0,Z=0),Scale3D=(X= ...(truncated)"
    WeightMinimum=0
    Weight=&RigVMFunction_MathBoolToFloat(Value=RigVMVariableNode() {variable=isParachuteOpen}.Value).Result
   )
   n11 = &RigVMFunction_MathQuaternionToEuler(
    RotationOrder=ZXY
    Value=&RigVMFunction_MathQuaternionSwingTwist(
     TwistAxis=(X=0,Y=0,Z=1)
     Input=&RigUnit_GetRelativeTransformForItem(Child=(Type=Bone,Name="head"), Parent=(Type=Bone,Name="spine_05"), bChildInitial=false, bParentInitial=false).RelativeTransform.Rotation
    ).Swing
   )
   AlphaInterp = &RigVMFunction_AlphaInterp(
    InterpSpeedDecreasing=10
    InterpSpeedIncreasing=10
    Scale=1
    bClampResult=False
    bInterpResult=False
    Value=RigVMVariableNode() {variable=isEmoting}.Value
   )
   RigVMFunction_AnimEvalRichCurve = &RigVMFunction_AnimEvalRichCurve(Curve=(EditorCurveData=(Keys=((InterpMode=RCIM_Cubic,Value=1),(Time=1)))), SourceMinimum=0, TargetMinimum=0, Value=@AlphaInterp.Result)
   Multiply = &RigVMFunction_MathDoubleMul(
    A=&RigVMFunction_MathDoubleRemap(SourceMaximum=-75, SourceMinimum=0, TargetMaximum=1, TargetMinimum=0, Value=@n11.Result.X).Result
    B=@RigVMFunction_AnimEvalRichCurve.Result
   )
   &RigUnit_ModifyTransforms(
    ItemToModify="((Item=(Type=Bone,Name=\"None\"),Transform=(Rotation=(X=0,Y=0,Z=0,W=1),Translation=(X=0,Y=0,Z=0),Scale3D=(X=1,Y=1,Z=1))),(Item=(Type=Bone,Name=\"base_hair_01_back\"),Transform=(Rotation=(X=-0.2892,Y=0.0848,Z=-0.1675,W=0.9387),Translation=(X=0,Y=0,Z=0),Scale3D= ...(truncated)"
    WeightMinimum=0
    Weight=&RigVMFunction_MathDoubleMul(B=0.5, A=@Multiply.Result).Result
    ExecutePin=@RigVMFunction_Sequence.C
   )
   n12 = &RigVMFunction_MathDoubleRemap(SourceMaximum=75, SourceMinimum=0, TargetMaximum=1, TargetMinimum=0, Value=@n11.Result.X)
   > &RigUnit_ModifyTransforms(
    ItemToModify="((Item=(Type=Bone,Name=\"None\"),Transform=(Rotation=(X=0,Y=0,Z=0,W=1),Translation=(X=0,Y=0,Z=0),Scale3D=(X=1,Y=1,Z=1))),(Item=(Type=Bone,Name=\"base_hair_01_back\"),Transform=(Rotation=(X=-0.2892,Y=0.0848,Z=-0.1675,W=0.9387),Translation=(X=0,Y=0,Z=0),Scale3D= ...(truncated)"
    WeightMinimum=0
    Weight=&RigVMFunction_MathDoubleMul(B=0.5, A=&RigVMFunction_MathDoubleMul(A=@n12.Result, B=@RigVMFunction_AnimEvalRichCurve.Result).Result).Result
   )
   &RigUnit_ModifyTransforms(
    ExecutePin=()
    ItemToModify="((Item=(Type=Bone,Name=\"None\"),Transform=(Rotation=(X=0,Y=0,Z=0,W=1),Translation=(X=0,Y=0,Z=0),Scale3D=(X=1,Y=1,Z=1))),(Item=(Type=Bone,Name=\"base_hair_01_lf\"),Transform=(Rotation=(X=0.3007,Y=0.0358,Z=-0.1782,W=0.9362),Translation=(X=0.5039,Y=-0.6261,Z=0.4 ...(truncated)"
    Weight=1
    WeightMinimum=0
   )
   RigVMVariableNode() {variable=neck_01_SPR_look_up_alpha}
  ]}
]
hierarchy [
 root {loc=<0 0 0>, rot=<0 0 0 1>}
 pelvis {parent=root, loc=<0 -0.7071 93.1191>, rot=<0.0127 0.707 -0.0127 -0.707>}
 spine_01 {parent=pelvis, loc=<6.7494 -0.6388 0>, rot=<0 0 0.034 -0.9994>}
 spine_02 {parent=spine_01, loc=<7.2518 -0.035 0>, rot=<0 0 0.0238 -0.9997>}
 spine_03 {parent=spine_02, loc=<6.0104 0 0>, rot=<0 0 0.0149 -0.9999>}
 spine_04 {parent=spine_03, loc=<8.1024 0 0>, rot=<0 0 -0.0132 -0.9999>}
 spine_05 {parent=spine_04, loc=<10.0146 0 0>, rot=<0 0 -0.0015 -1>}
 neck_01 {parent=spine_05, loc=<11.3138 0.1529 0>, rot=<0 0 -0.2003 -0.9797>}
 neck_02 {parent=neck_01, loc=<4.5627 0 0>, rot=<0 0 0.0504 -0.9987>}
 head {parent=neck_02, loc=<5.3914 0 0>, rot=<0 0 0.115 -0.9934>}
 base_hair_01_back {parent=head, loc=<7.7067 -7.5919 -0.2052>, rot=<0.032 -0.9995 0 0>}
 dyn_hair_01_back {parent=base_hair_01_back, loc=<0 0 0>, rot=<0 0 0 1>}
 dyn_hair_02_back {parent=dyn_hair_01_back, loc=<5.6066 0 0>, rot=<0 0 0.0227 -0.9997>}
 dyn_hair_03_back {parent=dyn_hair_02_back, loc=<4.9994 0 0>, rot=<0 0 0 1>}
 dyn_hair_04_back {parent=dyn_hair_03_back, loc=<4.1011 0 0>, rot=<0 0 0 1>}
 dyn_hair_04_back_helper {parent=dyn_hair_03_back, loc=<4.101 0 0>, rot=<0 0 0 1>}
 base_hair_01_lf {parent=head, loc=<5.1114 -3.0768 -6.7538>, rot=<-0.0206 -0.753 -0.6561 -0.0458>}
 dyn_hair_01_lf {parent=base_hair_01_lf, loc=<0 0 0>, rot=<0 0 0 1>}
 dyn_hair_02_lf {parent=dyn_hair_01_lf, loc=<4.1494 0 0>, rot=<-0.0001 0.1068 -0.0156 -0.9942>}
 dyn_hair_03_lf {parent=dyn_hair_02_lf, loc=<4.0327 0 0>, rot=<0 0 0 1>}
 base_hair_01_rt {parent=head, loc=<5.1114 -3.0768 6.7538>, rot=<-0.753 0.0206 0.0458 -0.6561>}
 dyn_hair_01_rt {parent=base_hair_01_rt, loc=<0 0 0>, rot=<0 0 0 1>}
 dyn_hair_02_rt {parent=dyn_hair_01_rt, loc=<-4.1492 0 0>, rot=<-0.0001 0.1068 -0.0156 -0.9942>}
 dyn_hair_03_rt {parent=dyn_hair_02_rt, loc=<-4.0327 0 0>, rot=<0 0 0 1>}
 FACIAL_C_FacialRoot {parent=head, loc=<2.1524 -0.746 -0.0002>, rot=<0.4879 -0.5118 -0.4879 -0.5118>}
 FACIAL_L_Masseter {parent=FACIAL_C_FacialRoot, loc=<6.091 0.983 4.3945>, rot=<0.0738 0.1129 0.0074 -0.9908>}
 FACIAL_L_Ear {parent=FACIAL_C_FacialRoot, loc=<6.725 -1.0964 1.2535>, rot=<-0.0218 -0.0009 -0.0261 -0.9994>}
 FACIAL_L_Jawline {parent=FACIAL_C_FacialRoot, loc=<4.5152 5.5563 5.0211>, rot=<0.2137 0.1574 -0.0553 -0.9626>}
 FACIAL_L_CheekOuter {parent=FACIAL_C_FacialRoot, loc=<3.5141 0.397 6.7987>, rot=<0.0593 -0.3765 0.1021 -0.9188>}
 FACIAL_L_CheekLower {parent=FACIAL_C_FacialRoot, loc=<4.0216 3.7655 6.0412>, rot=<0.0386 -0.4267 0.0597 -0.9016>}
 FACIAL_R_CheekOuter {parent=FACIAL_C_FacialRoot, loc=<-3.4919 0.3784 6.8064>, rot=<0.0593 0.3764 -0.102 -0.9189>}
 FACIAL_R_CheekLower {parent=FACIAL_C_FacialRoot, loc=<-3.9955 3.716 6.0538>, rot=<0.0387 0.4266 -0.0596 -0.9017>}
 FACIAL_R_Masseter {parent=FACIAL_C_FacialRoot, loc=<-6.0553 0.9685 4.4051>, rot=<0.0738 -0.113 -0.0073 -0.9908>}
 FACIAL_R_Ear {parent=FACIAL_C_FacialRoot, loc=<-6.7303 -1.097 1.2457>, rot=<-0.0218 0.0007 0.0262 -0.9994>}
 FACIAL_R_Jawline {parent=FACIAL_C_FacialRoot, loc=<-4.4645 5.5044 5.006>, rot=<0.2137 -0.1576 0.0554 -0.9625>}
 base_hair_01_back_lf {parent=head, loc=<7.564 -5.8151 -4.6286>, rot=<0.0296 -0.9259 -0.3763 -0.012>}
 dyn_hair_01_back_lf {parent=base_hair_01_back_lf, loc=<0 0 0>, rot=<0 0 0 1>}
 dyn_hair_02_back_lf {parent=dyn_hair_01_back_lf, loc=<5.6066 0 0>, rot=<0 0 0.0227 -0.9997>}
 dyn_hair_03_back_lf {parent=dyn_hair_02_back_lf, loc=<4.9994 0 0>, rot=<0 0 0 1>}
 base_hair_01_back_rt {parent=head, loc=<7.5641 -5.8151 4.6286>, rot=<0.9259 0.0296 -0.012 0.3763>}
 dyn_hair_01_back_rt {parent=base_hair_01_back_rt, loc=<0 0 0>, rot=<0 0 0 1>}
 dyn_hair_02_back_rt {parent=dyn_hair_01_back_rt, loc=<-5.6065 0 0>, rot=<0 0 0.0227 -0.9997>}
 dyn_hair_03_back_rt {parent=dyn_hair_02_back_rt, loc=<-4.9998 0 0>, rot=<0 0 0 1>}
 clavicle_l {parent=spine_05, loc=<4.3575 -0.1466 -2.3223>, rot=<0.0465 -0.7289 -0.009 -0.6829>}
 upperarm_l {parent=clavicle_l, loc=<15.5111 0 0>, rot=<0.0226 -0.3234 -0.0088 -0.9459>}
 clavicle_r {parent=spine_05, loc=<4.3575 -0.1466 2.3223>, rot=<0.7289 0.0465 -0.6829 0.009>}
 upperarm_r {parent=clavicle_r, loc=<-15.5111 0 0>, rot=<0.0226 -0.3234 -0.0088 -0.9459>}
]
variables [
 LongWTargeting {type=bool, default=False}
 PistolTaargeting {type=bool, default=False}
 clavicle_l_SPR_rotate_back {type=float, default=0}
 clavicle_r_SPR_rotate_back {type=float, default=0}
 head_L_twist {type=double, default=0}
 head_SPR_look_left_top_alpha {type=float, default=0}
 head_SPR_look_up_alpha {type=float, default=0}
 head_SPR_rot_back_alpha {type=float, default=0}
 isCrouching {type=bool, default=False}
 isEmoting {type=double, default=0}
 isFloating {type=bool, default=False}
 isParachuteOpen {type=bool, default=False}
 isSkydiving {type=bool, default=False}
 look_up_and_left_two_hands_multiplied_alpha {type=double, default=0}
 neck_01_SPR_look_up_alpha {type=float, default=0}
 upperarm_r_SPR_aim_up_two_hands_alpha {type=float, default=0}
]
