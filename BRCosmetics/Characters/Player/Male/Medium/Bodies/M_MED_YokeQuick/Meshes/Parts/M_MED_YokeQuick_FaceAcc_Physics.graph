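# Physics asset for the YokeQuick male character's FaceAcc mesh, setting up jointed hair against capsule colliders on the head, neck, upper spine and clavicles.
# Five hair chains (left, right, three back) each run from a base on the head through three twist-locked, narrowly swing-limited segments.
# Upper-body joints and most bases are angularly locked, but back-left and back-right bases rotate freely.
# Free-rotation constraints cross-link both back-side tips to the center-back tip.
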PhysicsAsset /BRCosmetics/Characters/Player/Male/Medium/Bodies/M_MED_YokeQuick/Meshes/Parts/M_MED_YokeQuick_FaceAcc_Physics
bodies {
 base_hair_01_back [{shapeType=Capsule, length=1, radius=1, shapeName=base_hair_01_back_capsule, rotation=<rot 0 -90 0>}]
 base_hair_01_back_lf [{shapeType=Capsule, length=1, radius=1, shapeName=base_hair_01_back_capsule, rotation=<rot 0 -90 0>}]
 base_hair_01_back_rt [{shapeType=Capsule, length=1, radius=1, shapeName=base_hair_01_back_capsule, rotation=<rot 0 -90 0>}]
 base_hair_01_lf [{shapeType=Capsule, length=1, radius=1, shapeName=base_hair_01_back_capsule, rotation=<rot 0 -90 0>}]
 base_hair_01_rt [{shapeType=Capsule, length=1, radius=1, shapeName=base_hair_01_back_capsule, rotation=<rot 0 -90 0>}]
 clavicle_l [{
   shapeType=Capsule
   length=12.5048
   radius=5.3842
   shapeName=clavicle_l_capsule
   center=<10.8379 -3.8274 2.3501>
   rotation=<rot -0.0008 -89.9999 105.329>
  }]
 clavicle_r [{
   shapeType=Capsule
   length=12.5048
   radius=5.3842
   shapeName=clavicle_l_capsule
   center=<-11.0197 3.8274 -3.0133>
   rotation=<rot -0.0008 -89.9999 105.329>
  }]
 dyn_hair_01_back [{shapeType=Capsule, length=4.7379, radius=2.2621, shapeName=dyn_hair_01_back_capsule, center=<2 0 0>}]
 dyn_hair_01_back_lf [{
   shapeType=Capsule
   length=4.7379
   radius=2.2621
   shapeName=dyn_hair_01_back_lf_capsule
   center=<2.0001 0 0>
   rotation=<rot 0.0168 0.0173 -4.4215>
  }]
 dyn_hair_01_back_rt [{shapeType=Capsule, length=2.8632, radius=2.2621, shapeName=dyn_hair_01_back_rt_capsule, center=<-2 0 0>, rotation=<rot 0 0 -5.9948>}]
 dyn_hair_01_lf [{
   shapeType=Capsule
   length=1.501
   radius=1.7613
   shapeName=dyn_hair_01_back_capsule
   center=<1.1444 0 0.0375>
   rotation=<rot 87.9258 -179.9916 -179.9925>
  }]
 dyn_hair_01_rt [{
   shapeType=Capsule
   length=2.1132
   radius=1.9073
   shapeName=dyn_hair_01_back_capsule
   center=<-1.1211 -0.0017 0.2936>
   rotation=<rot -85.9664 175.754 -172.7711>
  }]
 dyn_hair_02_back [{shapeType=Capsule, length=1.8524, radius=3.45, shapeName=dyn_hair_03_back_capsule, center=<1.8884 -2.316 0>}]
 dyn_hair_02_back_lf [{
   shapeType=Capsule
   length=4.7328
   radius=3.45
   shapeName=dyn_hair_03_back_lf_capsule
   center=<1.3628 -2.3805 -0.121>
   rotation=<rot 0.0168 0.0173 -4.4222>
  }]
 dyn_hair_02_back_rt [{
   shapeType=Capsule
   length=4.7328
   radius=3.45
   shapeName=dyn_hair_03_back_lf_capsule
   center=<-1.065 2.5079 0.2522>
   rotation=<rot 7.3561 -0.5533 -4.459>
  }]
 dyn_hair_02_lf [{
   shapeType=Capsule
   length=1.4014
   radius=1.8609
   shapeName=dyn_hair_01_back_capsule
   center=<1.1444 0 0.0375>
   rotation=<rot 87.9258 -179.9916 -179.9925>
  }]
 dyn_hair_02_rt [{
   shapeType=Capsule
   length=2.276
   radius=1.7445
   shapeName=dyn_hair_01_back_capsule
   center=<-1.0918 0.2897 0.3134>
   rotation=<rot -86.5817 179.3104 -178.9388>
  }]
 dyn_hair_03_back [{shapeType=Capsule, length=2.3842, radius=3.45, shapeName=dyn_hair_03_back_capsule, center=<1.8995 -1.8908 0>, rotation=<rot 0 -17.9393 0>}]
 dyn_hair_03_back_lf [{
   shapeType=Capsule
   length=4.7328
   radius=3.45
   shapeName=dyn_hair_03_back_lf_capsule
   center=<1.3628 -2.3805 -0.121>
   rotation=<rot 0.0168 0.0173 -4.4222>
  }]
 dyn_hair_03_back_rt [{
   shapeType=Capsule
   length=4.7328
   radius=3.45
   shapeName=dyn_hair_03_back_lf_capsule
   center=<-1.1812 2.955 0.2717>
   rotation=<rot 7.3561 -0.5533 -4.459>
  }]
 dyn_hair_03_lf [
  {
   shapeType=Capsule
   length=2.3761
   radius=1.9338
   shapeName=dyn_hair_03_back_capsule
   center=<2.2553 -0.5024 0.3434>
   rotation=<rot -71.5003 -2.269 -2.3697>
  }
  {shapeType=Capsule, length=1.72, radius=1.2178, shapeName=dyn_hair_03_lf_capsule, center=<3.5782 0 1.9229>, rotation=<rot 58.739 -180 -180>}
 ]
 dyn_hair_03_rt [
  {
   shapeType=Capsule
   length=4.5368
   radius=1.7707
   shapeName=dyn_hair_03_back_capsule
   center=<-1.5784 0.4695 -0.0886>
   rotation=<rot -60.9869 -1.5367 -5.6542>
  }
  {
   shapeType=Capsule
   length=1.6208
   radius=0.9755
   shapeName=dyn_hair_03_rt_capsule
   center=<-4.0018 -0.3864 -2.3135>
   rotation=<rot -44.5581 0 -10.0791>
  }
 ]
 head [{shapeType=Capsule, length=0, radius=8.9924, center=<3.2125 2.9665 -0.1299>, rotation=<rot 0 0 -90>}]
 neck_02 [{
   shapeType=Capsule
   length=4.9284
   radius=5.8613
   shapeName=neck_01_capsule
   center=<-0.0151 1.0689 0>
   rotation=<rot 0.0116 -86.0105 -90.1656>
  }]
 spine_05 [
  {shapeType=Capsule, length=19.6395, radius=9.8203, shapeName=spine_05_capsule, center=<3.5501 -0.8103 0>, rotation=<rot 0 -81.2455 0>}
  {shapeType=Capsule, length=10.1853, radius=7.3963, shapeName=spine_05_capsule, center=<7.6655 -1.0289 0>, rotation=<rot 0 15.2944 0>}
 ]
}
constraints [
 {
  bone1Name=base_hair_01_back
  bone2Name=head
  swing1LimitDegrees=45
  swing1Motion=Locked
  swing2LimitDegrees=45
  swing2Motion=Locked
  twistLimitDegrees=45
  twistMotion=Locked
 }
 {
  bone1Name=dyn_hair_01_back
  bone2Name=base_hair_01_back
  swing1LimitDegrees=10
  swing2LimitDegrees=15
  twistLimitDegrees=45
  twistMotion=Locked
 }
 {
  bone1Name=dyn_hair_02_back
  bone2Name=dyn_hair_01_back
  swing1LimitDegrees=10
  swing2LimitDegrees=15
  twistLimitDegrees=45
  twistMotion=Locked
 }
 {
  bone1Name=dyn_hair_03_back
  bone2Name=dyn_hair_02_back
  swing1LimitDegrees=12
  swing2LimitDegrees=15
  twistLimitDegrees=45
  twistMotion=Locked
 }
 {
  bone1Name=head
  bone2Name=neck_02
  swing1LimitDegrees=45
  swing1Motion=Locked
  swing2LimitDegrees=45
  swing2Motion=Locked
  twistLimitDegrees=45
  twistMotion=Locked
 }
 {
  bone1Name=neck_02
  bone2Name=spine_05
  swing1LimitDegrees=45
  swing1Motion=Locked
  swing2LimitDegrees=45
  swing2Motion=Locked
  twistLimitDegrees=45
  twistMotion=Locked
 }
 {
  bone1Name=clavicle_l
  bone2Name=spine_05
  swing1LimitDegrees=45
  swing1Motion=Locked
  swing2LimitDegrees=45
  swing2Motion=Locked
  twistLimitDegrees=45
  twistMotion=Locked
 }
 {
  bone1Name=clavicle_r
  bone2Name=spine_05
  swing1LimitDegrees=45
  swing1Motion=Locked
  swing2LimitDegrees=45
  swing2Motion=Locked
  twistLimitDegrees=45
  twistMotion=Locked
 }
 {
  bone1Name=base_hair_01_lf
  bone2Name=head
  swing1LimitDegrees=45
  swing1Motion=Locked
  swing2LimitDegrees=45
  swing2Motion=Locked
  twistLimitDegrees=45
  twistMotion=Locked
 }
 {bone1Name=dyn_hair_01_lf, bone2Name=base_hair_01_lf, swing1LimitDegrees=10, swing2LimitDegrees=20, twistLimitDegrees=45, twistMotion=Locked}
 {bone1Name=dyn_hair_02_lf, bone2Name=dyn_hair_01_lf, swing1LimitDegrees=10, swing2LimitDegrees=20, twistLimitDegrees=45, twistMotion=Locked}
 {bone1Name=dyn_hair_03_lf, bone2Name=dyn_hair_02_lf, swing1LimitDegrees=10, swing2LimitDegrees=20, twistLimitDegrees=45, twistMotion=Locked}
 {
  bone1Name=base_hair_01_rt
  bone2Name=head
  swing1LimitDegrees=45
  swing1Motion=Locked
  swing2LimitDegrees=45
  swing2Motion=Locked
  twistLimitDegrees=45
  twistMotion=Locked
 }
 {bone1Name=dyn_hair_01_rt, bone2Name=base_hair_01_rt, swing1LimitDegrees=10, swing2LimitDegrees=20, twistLimitDegrees=45, twistMotion=Locked}
 {bone1Name=dyn_hair_02_rt, bone2Name=dyn_hair_01_rt, swing1LimitDegrees=10, swing2LimitDegrees=20, twistLimitDegrees=45, twistMotion=Locked}
 {bone1Name=dyn_hair_03_rt, bone2Name=dyn_hair_02_rt, swing1LimitDegrees=12, swing2LimitDegrees=20, twistLimitDegrees=45, twistMotion=Locked}
 {
  bone1Name=base_hair_01_back_lf
  bone2Name=head
  swing1LimitDegrees=45
  swing1Motion=Free
  swing2LimitDegrees=45
  swing2Motion=Free
  twistLimitDegrees=45
  twistMotion=Free
 }
 {
  bone1Name=dyn_hair_01_back_lf
  bone2Name=base_hair_01_back_lf
  swing1LimitDegrees=10
  swing2LimitDegrees=15
  twistLimitDegrees=45
  twistMotion=Locked
 }
 {
  bone1Name=dyn_hair_02_back_lf
  bone2Name=dyn_hair_01_back_lf
  swing1LimitDegrees=10
  swing2LimitDegrees=15
  twistLimitDegrees=45
  twistMotion=Locked
 }
 {
  bone1Name=dyn_hair_03_back_lf
  bone2Name=dyn_hair_02_back_lf
  swing1LimitDegrees=12
  swing2LimitDegrees=15
  twistLimitDegrees=45
  twistMotion=Locked
 }
 {
  bone1Name=base_hair_01_back_rt
  bone2Name=head
  swing1LimitDegrees=45
  swing1Motion=Free
  swing2LimitDegrees=45
  swing2Motion=Free
  twistLimitDegrees=45
  twistMotion=Free
 }
 {
  bone1Name=dyn_hair_01_back_rt
  bone2Name=base_hair_01_back_rt
  swing1LimitDegrees=10
  swing2LimitDegrees=15
  twistLimitDegrees=45
  twistMotion=Locked
 }
 {
  bone1Name=dyn_hair_02_back_rt
  bone2Name=dyn_hair_01_back_rt
  swing1LimitDegrees=10
  swing2LimitDegrees=15
  twistLimitDegrees=45
  twistMotion=Locked
 }
 {
  bone1Name=dyn_hair_03_back_rt
  bone2Name=dyn_hair_02_back_rt
  swing1LimitDegrees=12
  swing2LimitDegrees=15
  twistLimitDegrees=45
  twistMotion=Locked
 }
 {
  bone1Name=dyn_hair_03_back_lf
  bone2Name=dyn_hair_03_back
  swing1LimitDegrees=45
  swing1Motion=Free
  swing2LimitDegrees=45
  swing2Motion=Free
  twistLimitDegrees=45
  twistMotion=Free
 }
 {
  bone1Name=dyn_hair_03_back_rt
  bone2Name=dyn_hair_03_back
  swing1LimitDegrees=45
  swing1Motion=Free
  swing2LimitDegrees=45
  swing2Motion=Free
  twistLimitDegrees=45
  twistMotion=Free
 }
]
