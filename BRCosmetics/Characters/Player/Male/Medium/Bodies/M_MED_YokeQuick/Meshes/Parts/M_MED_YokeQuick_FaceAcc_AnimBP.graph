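# Animation Blueprint for the YokeQuick body's face accessory part, adding secondary motion to its dynamic hair.
# The parent mesh pose feeds a Control Rig driven by skydive, crouch, emote and weapon-aiming flags, then head-space RigidBody physics and additive right-hair rotations.
# Each update derives those weapon flags from the owner's FortPlayerAnimInstance.
# It also computes gravity-override physics settings, though the RigidBody node uses fixed values and the left-hair offset is unconnected.
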AnimBlueprint /BRCosmetics/Characters/Player/Male/Medium/Bodies/M_MED_YokeQuick/Meshes/Parts/M_MED_YokeQuick_FaceAcc_AnimBP
target_skeleton=&/M_MED_YokeQuick_FaceAcc_Skeleton
graphs [
 AnimGraph {nodes [
   "Transform (Modify) Bone - Bone: dyn_hair_02_lf"(Alpha=1, Translation="0,0,0", Rotation="-1.860,-14.286,-12.625", Scale="1,1,1") {
    rotationMode=BMM_Additive
    rotationSpace=BCS_BoneSpace
    translationMode=BMM_Additive
    translationSpace=BCS_BoneSpace
    boneToModify {boneName=dyn_hair_02_lf}
    rotation=<rot -1.86 -14.286 -12.625>
   }
   n1 = RigidBody!(
    ComponentPose=Local To Component(LocalPose=Use cached pose 'CR_ControlRig'())
    Alpha
    OverrideWorldGravity:Vector="0,0,-980"
    ExternalForce
    SimSpaceSettings:Sim Space Settings Structure=(WorldAlpha=1,VelocityScaleZ=0,DampingAlpha=1,MaxLinearVelocity=500,MaxAngularVelocity=3,MaxLinearAcceleration=700,MaxAngularAcceleration=1000,ExternalLinearDragV=(X=0.1,Y=0.1,Z=0),ExternalLinearVelocity=(X=0,Y=0,Z=0),ExternalAngularVelocity=(X=0,Y=0,Z=0))
   ) -> (Pose) {
    bClampLinearTranslationLimitToRefPose=true
    bOverrideWorldGravity=true
    overridePhysicsAsset=&/M_MED_YokeQuick_FaceAcc_Physics
    simulationSpace=BaseBoneSpace
    alphaBoolBlend {blendOption=Linear}
    baseBoneRef {boneName=head}
    componentAppliedLinearAccClamp=<3000 3000 3000>
    componentLinearAccScale=<0.1 0.1 0.1>
    componentLinearVelScale=<0.2 0.2 0.2>
    overrideWorldGravity {z=-980}
    simSpaceSettings {
     maxAngularAcceleration=1000
     maxAngularVelocity=3
     maxLinearAcceleration=700
     maxLinearVelocity=500
     velocityScaleZ=0
     worldAlpha=1
     externalLinearDragV=<0.1 0.1>
    }
   }
   n2 = "Transform (Modify) Bone - Bone: dyn_hair_02_rt"(ComponentPose=@n1, Alpha=1, Translation="0,0,0", Rotation="10.765,22.632,-15.188", Scale="1,1,1") {
    rotationMode=BMM_Additive
    rotationSpace=BCS_BoneSpace
    translationMode=BMM_Additive
    translationSpace=BCS_BoneSpace
    boneToModify {boneName=dyn_hair_02_rt}
    rotation=<rot 10.765 22.632 -15.188>
   }
   n3 = "Transform (Modify) Bone - Bone: dyn_hair_02_rt"(ComponentPose=@n2, Alpha=1, Translation="0,0,0", Rotation="24.291,37.032,12.810", Scale="1,1,1") {
    rotationMode=BMM_Additive
    rotationSpace=BCS_BoneSpace
    translationMode=BMM_Additive
    translationSpace=BCS_BoneSpace
    boneToModify {boneName=dyn_hair_02_rt}
    rotation=<rot 24.291 37.032 12.81>
   }
   Output Pose(Result=Component To Local(ComponentPose=@n3)) {layerGroup=None}
   n4 = Control Rig!(
    Source=Copy Pose From Mesh() {bUseAttachedParent=true}
    Alpha
    isSkydiving:bool=Get bIsSkyDiving()
    isCrouching:bool=Get bIsCrouching()
    LongWTargeting:bool=Get LongW Equipped() -> (LongW Equipped:bool)
    PistolTaargeting:bool=Get Pistol Equipped() -> (Pistol Equipped:bool)
    isFloating:bool=Get bIsSkydiveFloating()
    isEmoting:double=Get PlayingEmote() -> (PlayingEmote:double)
    isParachuteOpen:bool=Get bIsParachuteOpened()
   ) -> (Pose) {controlRigAssetReference {blueprintRigClass=&/M_MED_YokeQuick_FaceAcc_CtrlRig.M_MED_YokeQuick_FaceAcc_CtrlRig_C}}
   Save cached pose 'CR_ControlRig'(Pose=@n4) {cacheName=CR_ControlRig}
   Get bIsPlayingEmote()
  ]}
 EventGraph {nodes [
   Event BlueprintUpdateAnimation()
   > n1 = Sequence!(execute) -> (then_0, then_1, then_2:Exec)
   > n2 = "Gravity Override State Function (Parameter Asset) (Nativized)"(
    CharacterPartAnimInstance=Self-Reference()
    GameTimeInSeconds=GetGameTimeInSeconds()
    DebugJointName=spine_05
    Parameters=&/M_MED_YokeQuick_FaceAcc_Parameters
   )
   > Set Gravity Override(Gravity Override=@n2)
   > Set RBAN Settings Out(RBAN Settings Out=@n2.OutRigidBodyAnimNodeParameters)
   n3 = Branch(execute=@n1.then_2, Condition=Get bIsPlayingEmote())
   > Set PlayingEmote(PlayingEmote:double=1) -> (Output_Get:double)
   Set PlayingEmote(execute=@n3.else, PlayingEmote=0)
   n4 = Cast To FortPlayerAnimInstance(execute=@n1.then_1, Object=GetAnimInstance(self=GetOwnerSkeletalMeshComponent()))
   > Set LongW Equipped(LongW Equipped:bool=false) -> (Output_Get:bool)
   > n5 = Switch on EFortWeaponCoreAnimation!(execute, Selection:EFortWeaponCoreAnimation Enum=Get WeaponCoreAnim(self=@n4)) -> (
    Melee:Exec
    Pistol:Exec
    Shotgun:Exec
    PaperBlueprint:Exec
    Rifle:Exec
    MeleeOneHand:Exec
    MachinePistol:Exec
    RocketLauncher:Exec
    GrenadeLauncher:Exec
    GoingCommando:Exec
    AssaultRifle:Exec
    TacticalShotgun:Exec
    SniperRifle:Exec
    TrapPlacement:Exec
    ShoulderLauncher:Exec
    AbilityDecoTool:Exec
    Crossbow:Exec
    C4:Exec
    RemoteControl:Exec
    DualWield:Exec
    AR_BullPup:Exec
    AR_ForwardGrip:Exec
    MedPackPaddles:Exec
    SMG_P90:Exec
    AR_DrumGun:Exec
    Consumable_Small:Exec
    Consumable_Large:Exec
    Balloon:Exec
    MountedTurret:Exec
    CreativeTool:Exec
    ExplosiveBow:Exec
    AshtonIndigo:Exec
    AshtonChicago:Exec
    MeleeDualWield:Exec
    Unarmed:Exec
   )
   n6 = Branch(
    execute=[@n5.AssaultRifle, @n5.GoingCommando, @n5.GrenadeLauncher, @n5.Rifle, @n5.Shotgun, @n5.SniperRifle, @n5.TacticalShotgun]
    Condition=Get bIsTargeting()
   )
   > Set LongW Equipped(LongW Equipped=true)
   Set LongW Equipped(execute=@n6.else, LongW Equipped=false)
   n7 = Branch(execute=[@n5.DualWield, @n5.Pistol], Condition=Get bIsTargeting())
   > Set Pistol Equipped(Pistol Equipped:bool=true) -> (Output_Get:bool)
   Set Pistol Equipped(execute=@n7.else, Pistol Equipped=false)
  ]}
]
variables [
 Gravity Override {type=Vector}
 LongW Equipped {type=boolean}
 Pistol Equipped {type=boolean}
 PlayingEmote {type=number}
 RBAN Settings Out {type=FortRigidBodyAnimNodeParameters}
 Skydive {type=number}
]
